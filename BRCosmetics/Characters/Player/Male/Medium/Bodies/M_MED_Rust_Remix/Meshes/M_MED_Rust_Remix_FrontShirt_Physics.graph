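# Physics asset for the front shirt of the M_MED_Rust_Remix male medium body cosmetic, simulating its dangling cloth with bone chains.
# Center, left and right chains of four capsules each hang from tiny MSTR anchor bodies constrained to spine_01, with larger capsules on the lower links.
# Chain joints use angular limits, while anchor and body joints swing and twist freely.
# Pelvis, spine and thigh capsules approximate the body beneath the shirt.
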PhysicsAsset /BRCosmetics/Characters/Player/Male/Medium/Bodies/M_MED_Rust_Remix/Meshes/M_MED_Rust_Remix_FrontShirt_Physics
bodies {
 dyn_shirt_front_01 [{shapeType=Capsule, length=5.5944, radius=4.0599, shapeName=dyn_shirt_front_01_capsule, center=<7.0802 0 3>, rotation=<rot 0 0 90>}]
 dyn_shirt_front_01_MSTR [{shapeType=Capsule, length=0.505, radius=0.505, shapeName=dyn_shirt_front_01_MSTR_capsule, rotation=<rot 0 -90 0>}]
 dyn_shirt_front_01_l [{shapeType=Capsule, length=10.0347, radius=3.9653, shapeName=dyn_shirt_front_01_l_capsule, center=<8.5477 0 3>, rotation=<rot 0 0 90>}]
 dyn_shirt_front_01_l_MSTR [{shapeType=Capsule, length=0.505, radius=0.505, shapeName=dyn_shirt_front_01_l_MSTR_capsule, rotation=<rot 0 -90 0>}]
 dyn_shirt_front_01_r [{shapeType=Capsule, length=9.9401, radius=4.0599, shapeName=dyn_shirt_front_01_r_capsule, center=<7.0802 0 3>, rotation=<rot 0 0 90>}]
 dyn_shirt_front_01_r_MSTR [{shapeType=Capsule, length=0.505, radius=0.505, shapeName=dyn_shirt_front_01_r_MSTR_capsule, rotation=<rot 0 -90 0>}]
 dyn_shirt_front_02 [{shapeType=Capsule, length=5.5944, radius=4.0599, shapeName=dyn_shirt_front_02_capsule, center=<7.0802 0 3>, rotation=<rot 0 0 90>}]
 dyn_shirt_front_02_l [{shapeType=Capsule, length=10.0347, radius=3.9653, shapeName=dyn_shirt_front_02_l_capsule, center=<8.5477 0 3>, rotation=<rot 0 0 90>}]
 dyn_shirt_front_02_r [{shapeType=Capsule, length=9.9401, radius=4.0599, shapeName=dyn_shirt_front_02_r_capsule, center=<7.0802 0 3>, rotation=<rot 0 0 90>}]
 dyn_shirt_front_03 [{shapeType=Capsule, length=14.0836, radius=6.7988, shapeName=dyn_shirt_front_03_capsule, center=<7.0802 0 3>, rotation=<rot 0 0 90>}]
 dyn_shirt_front_03_l [{shapeType=Capsule, length=11.7977, radius=6.7988, shapeName=dyn_shirt_front_03_l_capsule, center=<7.0802 0 3>, rotation=<rot 0 0 90>}]
 dyn_shirt_front_03_r [{shapeType=Capsule, length=11.7977, radius=6.7988, shapeName=dyn_shirt_front_03_r_capsule, center=<7.0802 0 3>, rotation=<rot 0 0 90>}]
 dyn_shirt_front_04 [{shapeType=Capsule, length=14.0836, radius=6.7988, shapeName=dyn_shirt_front_04_capsule, center=<5.5009 0 3>, rotation=<rot 0 0 90>}]
 dyn_shirt_front_04_l [{shapeType=Capsule, length=11.7977, radius=6.7988, shapeName=dyn_shirt_front_04_l_capsule, center=<5.5009 0 3>, rotation=<rot 0 0 90>}]
 dyn_shirt_front_04_r [{shapeType=Capsule, length=11.7977, radius=6.7988, shapeName=dyn_shirt_front_04_r_capsule, center=<5.5009 0 3>, rotation=<rot 0 0 90>}]
 pelvis [
  {shapeType=Capsule, length=9.2986, radius=12.2012, shapeName=pelvis_capsule}
  {shapeType=Capsule, length=9.2986, radius=12.2012, shapeName=pelvis_capsule, center=<-13.8391 1.0166 0>}
 ]
 spine_01 [{shapeType=Capsule, length=10.3506, radius=11.1492, shapeName=spine_01_capsule}]
 spine_02 [{shapeType=Capsule, length=10.3506, radius=11.1492, shapeName=spine_02_capsule}]
 spine_03 [{shapeType=Capsule, length=10.3506, radius=11.1492, shapeName=spine_03_capsule}]
 thigh_l [{shapeType=Capsule, length=42.4439, radius=9.5199, shapeName=thigh_l_capsule, center=<-18.4161 0.0616 0>, rotation=<rot 90 90 85.8387>}]
 thigh_r [{shapeType=Capsule, length=42.4439, radius=9.5199, shapeName=thigh_r_capsule, center=<16.9113 -0.1522 0>, rotation=<rot 90 90 85.8387>}]
}
constraints [
 {
  bone1Name=spine_01
  bone2Name=pelvis
  swing1LimitDegrees=45
  swing1Motion=Free
  swing2LimitDegrees=45
  swing2Motion=Free
  twistLimitDegrees=45
  twistMotion=Free
 }
 {
  bone1Name=spine_02
  bone2Name=spine_01
  swing1LimitDegrees=45
  swing1Motion=Free
  swing2LimitDegrees=45
  swing2Motion=Free
  twistLimitDegrees=45
  twistMotion=Free
 }
 {
  bone1Name=spine_03
  bone2Name=spine_02
  swing1LimitDegrees=45
  swing1Motion=Free
  swing2LimitDegrees=45
  swing2Motion=Free
  twistLimitDegrees=45
  twistMotion=Free
 }
 {
  bone1Name=thigh_l
  bone2Name=pelvis
  swing1LimitDegrees=45
  swing1Motion=Free
  swing2LimitDegrees=45
  swing2Motion=Free
  twistLimitDegrees=45
  twistMotion=Free
 }
 {
  bone1Name=thigh_r
  bone2Name=pelvis
  swing1LimitDegrees=45
  swing1Motion=Free
  swing2LimitDegrees=45
  swing2Motion=Free
  twistLimitDegrees=45
  twistMotion=Free
 }
 {
  bone1Name=dyn_shirt_front_01_MSTR
  bone2Name=spine_01
  swing1LimitDegrees=45
  swing1Motion=Free
  swing2LimitDegrees=45
  swing2Motion=Free
  twistLimitDegrees=45
  twistMotion=Free
 }
 {
  bone1Name=dyn_shirt_front_01_r_MSTR
  bone2Name=spine_01
  swing1LimitDegrees=45
  swing1Motion=Free
  swing2LimitDegrees=45
  swing2Motion=Free
  twistLimitDegrees=45
  twistMotion=Free
 }
 {
  bone1Name=dyn_shirt_front_01_l_MSTR
  bone2Name=spine_01
  swing1LimitDegrees=45
  swing1Motion=Free
  swing2LimitDegrees=45
  swing2Motion=Free
  twistLimitDegrees=45
  twistMotion=Free
 }
 {bone1Name=dyn_shirt_front_01, bone2Name=dyn_shirt_front_01_MSTR, swing1LimitDegrees=10, swing2LimitDegrees=45, twistLimitDegrees=15}
 {bone1Name=dyn_shirt_front_02, bone2Name=dyn_shirt_front_01, swing1LimitDegrees=10, swing2LimitDegrees=45, twistLimitDegrees=15}
 {bone1Name=dyn_shirt_front_03, bone2Name=dyn_shirt_front_02, swing1LimitDegrees=10, swing2LimitDegrees=45, twistLimitDegrees=15}
 {bone1Name=dyn_shirt_front_04, bone2Name=dyn_shirt_front_03, swing1LimitDegrees=10, swing2LimitDegrees=45, twistLimitDegrees=15}
 {bone1Name=dyn_shirt_front_01_r, bone2Name=dyn_shirt_front_01_r_MSTR, swing1LimitDegrees=10, swing2LimitDegrees=45, twistLimitDegrees=15}
 {bone1Name=dyn_shirt_front_02_r, bone2Name=dyn_shirt_front_01_r, swing1LimitDegrees=10, swing2LimitDegrees=45, twistLimitDegrees=15}
 {bone1Name=dyn_shirt_front_03_r, bone2Name=dyn_shirt_front_02_r, swing1LimitDegrees=10, swing2LimitDegrees=45, twistLimitDegrees=15}
 {bone1Name=dyn_shirt_front_04_r, bone2Name=dyn_shirt_front_03_r, swing1LimitDegrees=10, swing2LimitDegrees=45, twistLimitDegrees=15}
 {bone1Name=dyn_shirt_front_01_l, bone2Name=dyn_shirt_front_01_l_MSTR, swing1LimitDegrees=10, swing2LimitDegrees=45, twistLimitDegrees=15}
 {bone1Name=dyn_shirt_front_02_l, bone2Name=dyn_shirt_front_01_l, swing1LimitDegrees=10, swing2LimitDegrees=45, twistLimitDegrees=15}
 {bone1Name=dyn_shirt_front_03_l, bone2Name=dyn_shirt_front_02_l, swing1LimitDegrees=10, swing2LimitDegrees=45, twistLimitDegrees=15}
 {bone1Name=dyn_shirt_front_04_l, bone2Name=dyn_shirt_front_03_l, swing1LimitDegrees=10, swing2LimitDegrees=45, twistLimitDegrees=15}
]
