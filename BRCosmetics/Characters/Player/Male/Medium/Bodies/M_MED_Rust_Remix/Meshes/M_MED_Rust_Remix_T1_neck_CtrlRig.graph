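# Control Rig for the neck of the M_MED_Rust_Remix cosmetic male body, offsetting collar bones beside the head.
# A sequence runs two branches, each a head-driven spherical pose reader followed by two Modify Transforms nodes.
# These shift the left, then right, dyn_collar1_02 and dyn_collar2_02 bones along Z.
# Both readers have debug drawing on, but their outputs are not wired to the offsets, whose weights are fixed at full.
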ControlRigBlueprint /BRCosmetics/Characters/Player/Male/Medium/Bodies/M_MED_Rust_Remix/Meshes/M_MED_Rust_Remix_T1_neck_CtrlRig
graphs [
 RigVMFunctionLibrary {links [], nodes []}
 RigVMModel {nodes [
   RigUnit_BeginExecution = &RigUnit_BeginExecution()
   n1 = /Script/ControlRig.RigUnit_SequenceExecution(ExecuteContext=@RigUnit_BeginExecution.ExecutePin) {is=struct RigVMUnitNode}
   &RigUnit_SphericalPoseReader(
    ActiveRegionScaleFactors=(PositiveWidth=1,NegativeWidth=1,PositiveHeight=1,NegativeHeight=1)
    ActiveRegionSize=0.1
    Debug=(bDrawDebug=True,DebugScale=25,DebugSegments=20,DebugThickness=0.25)
    DriverAxis=(X=0,Y=1,Z=0)
    DriverItem=(Type=Bone,Name="head")
    FalloffRegionScaleFactors=(PositiveWidth=1,NegativeWidth=1,PositiveHeight=1,NegativeHeight=0.6)
    FalloffSize=0.2
    OptionalParentItem=(Type=Bone)
    RotationOffset=(X=120,Y=0,Z=0)
    ExecutePin=@n1.A
   )
   > &RigUnit_ModifyTransforms(
    ItemToModify=((Item=(Type=Bone,Name="dyn_collar2_02_l"),Transform=(Rotation=(X=0,Y=0,Z=0,W=1),Translation=(X=0,Y=0,Z=-2),Scale3D=(X=1,Y=1,Z=1))))
    Weight=1
   )
   > &RigUnit_ModifyTransforms(
    ItemToModify=((Item=(Type=Bone,Name="dyn_collar1_02_l"),Transform=(Rotation=(X=0,Y=0,Z=0,W=1),Translation=(X=0,Y=0,Z=-3),Scale3D=(X=1,Y=1,Z=1))))
    Weight=1
   )
   &RigUnit_SphericalPoseReader(
    ActiveRegionScaleFactors=(PositiveWidth=1,NegativeWidth=1,PositiveHeight=1,NegativeHeight=1)
    ActiveRegionSize=0.1
    Debug=(bDrawDebug=True,DebugScale=25,DebugSegments=20,DebugThickness=0.25)
    DriverAxis=(X=0,Y=1,Z=0)
    DriverItem=(Type=Bone,Name="head")
    FalloffRegionScaleFactors=(PositiveWidth=1,NegativeWidth=1,PositiveHeight=0.6,NegativeHeight=1)
    FalloffSize=0.2
    OptionalParentItem=(Type=Bone)
    RotationOffset=(X=75,Y=-20,Z=0)
    ExecutePin=@n1.B
   )
   > &RigUnit_ModifyTransforms(
    ItemToModify=((Item=(Type=Bone,Name="dyn_collar2_02_r"),Transform=(Rotation=(X=0,Y=0,Z=0,W=1),Translation=(X=0,Y=0,Z=1.5),Scale3D=(X=1,Y=1,Z=1))))
    Weight=1
   )
   > &RigUnit_ModifyTransforms(
    ItemToModify=((Item=(Type=Bone,Name="dyn_collar1_02_r"),Transform=(Rotation=(X=0,Y=0,Z=0,W=1),Translation=(X=0,Y=0,Z=4),Scale3D=(X=1,Y=1,Z=1))))
    Weight=1
   )
  ]}
]
hierarchy [
 root {loc=<0 0 0>, rot=<0 0 0 1>}
 attach {parent=root, loc=<0 0 0>, rot=<0 0 0 1>}
 pelvis {parent=root, loc=<0 -0.7071 93.1191>, rot=<0.0127 0.707 -0.0127 -0.707>}
 spine_01 {parent=pelvis, loc=<6.7494 -0.6388 0>, rot=<0 0 0.034 -0.9994>}
 spine_02 {parent=spine_01, loc=<7.2518 -0.035 0>, rot=<0 0 0.0238 -0.9997>}
 spine_03 {parent=spine_02, loc=<6.0104 0 0>, rot=<0 0 0.0149 -0.9999>}
 spine_04 {parent=spine_03, loc=<8.1024 0 0>, rot=<0 0 -0.0132 -0.9999>}
 spine_05 {parent=spine_04, loc=<10.0146 0 0>, rot=<0 0 -0.0015 -1>}
 clavicle_l {parent=spine_05, loc=<4.3575 -0.1466 -2.3223>, rot=<0.0465 -0.7289 -0.009 -0.6829>}
 upperarm_l {parent=clavicle_l, loc=<15.5111 0 0>, rot=<0.0226 -0.3234 -0.0088 -0.9459>}
 lowerarm_l {parent=upperarm_l, loc=<26.6514 0 0>, rot=<0.0391 -0.0099 -0.2453 -0.9686>}
 hand_l {parent=lowerarm_l, loc=<22.3491 0 0>, rot=<-0.7081 -0.0507 0.0975 -0.6975>}
 index_metacarpal_l {parent=hand_l, loc=<1.6535 0.0183 -2.96>, rot=<0.0158 -0.0124 -0.0173 -0.9996>}
 index_01_l {parent=index_metacarpal_l, loc=<8.3008 0.0357 -0.0524>, rot=<-0.0044 0.0485 0.0924 -0.9945>}
 index_02_l {parent=index_01_l, loc=<3.3587 -0.3004 0.0452>, rot=<-0.002 0.0027 0.3225 -0.9466>}
 index_03_l {parent=index_02_l, loc=<2.5578 -0.123 -0.0252>, rot=<-0.002 0.0067 0.2558 -0.9667>}
 middle_metacarpal_l {parent=hand_l, loc=<2.1856 -0.1725 -0.5984>, rot=<0.006 0.0168 -0.0329 -0.9993>}
 middle_01_l {parent=middle_metacarpal_l, loc=<7.3275 -0.2447 0.0129>, rot=<-0.0208 0.0042 0.0858 -0.9961>}
 middle_02_l {parent=middle_01_l, loc=<4.0123 0.1401 0.1175>, rot=<-0.0218 0.0258 0.3512 -0.9357>}
 middle_03_l {parent=middle_02_l, loc=<2.6077 0.0534 0.1373>, rot=<0.0082 0.0188 0.1708 -0.9851>}
 pinky_metacarpal_l {parent=hand_l, loc=<1.7616 -0.9753 3.3814>, rot=<-0.0397 0.0814 -0.069 -0.9935>}
 pinky_01_l {parent=pinky_metacarpal_l, loc=<7.1709 -0.4831 -0.1094>, rot=<-0.0654 -0.0035 0.2242 -0.9723>}
 pinky_02_l {parent=pinky_01_l, loc=<2.8274 0.0893 -0.1043>, rot=<0.0037 0.0011 0.2854 -0.9584>}
 pinky_03_l {parent=pinky_02_l, loc=<2.2046 0.0307 -0.0078>, rot=<-0.0002 0.0098 0.1445 -0.9895>}
 ring_metacarpal_l {parent=hand_l, loc=<2.2074 -0.3625 1.3638>, rot=<-0.0284 0.0722 -0.0458 -0.9959>}
 ring_01_l {parent=ring_metacarpal_l, loc=<7.0295 -0.6197 -0.1446>, rot=<-0.0489 -0.0252 0.1892 -0.9804>}
 ring_02_l {parent=ring_01_l, loc=<3.6174 0.4647 -0.1027>, rot=<0 0 0.2806 -0.9598>}
 ring_03_l {parent=ring_02_l, loc=<2.4152 0.114 0.0199>, rot=<0 0 0.1189 -0.9929>}
 thumb_01_l {parent=hand_l, loc=<1.3118 -0.3191 -2.459>, rot=<0.6546 -0.4042 0.0664 -0.6354>}
 thumb_02_l {parent=thumb_01_l, loc=<5.9336 -0.0448 0.0401>, rot=<-0.0037 0.0213 0.298 -0.9543>}
 thumb_03_l {parent=thumb_02_l, loc=<2.2223 0.3682 -0.0907>, rot=<0.0118 -0.013 0.0066 -0.9998>}
 weapon_l {parent=hand_l, loc=<-0.1974 -3.4111 0.3811>, rot=<0 0 0 1>}
 lowerarm_twist_01_l {parent=lowerarm_l, loc=<17.8549 0 0>, rot=<0 0 0 1>}
 dyn_wristFlap_01_l_MSTR {parent=lowerarm_twist_01_l, loc=<0.5564 -3.5324 2.8082>, rot=<0.1255 0.7303 -0.0562 -0.6692>}
 dyn_wristFlap_01_l {parent=dyn_wristFlap_01_l_MSTR, loc=<0.2 0 0>, rot=<0 0 0 1>}
 lowerarm_twist_02_l {parent=lowerarm_l, loc=<9.6199 0 0>, rot=<0 0 0 1>}
 upperarm_twist_01_l {parent=upperarm_l, loc=<0.0041 0 0>, rot=<0 0 0 1>}
 shoulderpad_01_l_DSP {parent=upperarm_twist_01_l, loc=<1.5184 0 0>, rot=<0 0 0 1>}
 shoulderpad_02_l_DSP {parent=shoulderpad_01_l_DSP, loc=<0 0 5.864>, rot=<0 0 0 1>}
 upperarm_jacket_01_l_DSP {parent=upperarm_twist_01_l, loc=<0.0024 0.0112 -0.0019>, rot=<0 0 0 1>}
 upperarm_twist_02_l {parent=upperarm_l, loc=<13.7941 0 0>, rot=<0 0 0 1>}
 dyn_parent_shldr_pad_l {parent=upperarm_l, loc=<4.8682 0 0.0003>, rot=<-0.0001 0.0904 0.0046 -0.9959>}
 dyn_shldr_pad_base_l {parent=dyn_parent_shldr_pad_l, loc=<0 0 0>, rot=<0 0 0 1>}
 dyn_shldr_pad_bot_l {parent=dyn_shldr_pad_base_l, loc=<10.753 -0.1309 2.3495>, rot=<-0.0264 -0.7092 -0.7023 0.0556>}
 dyn_shldr_pad_top_l {parent=dyn_shldr_pad_bot_l, loc=<20.0911 1.3745 0>, rot=<0 0 0 1>}
 dyn_shldr_frt_l {parent=upperarm_l, loc=<4.7486 6.0026 0.9201>, rot=<-0.0001 0.0904 0.0046 -0.9959>}
 dyn_shldr_bck_l {parent=upperarm_l, loc=<4.8584 -7.2332 1.0675>, rot=<-0.0001 0.0904 0.0046 -0.9959>}
 dyn_shldr_aim_l {parent=upperarm_l, loc=<21.1633 0 -0.0002>, rot=<0 0 0 1>}
 shoulderBodyArmor_01_l_DSP_MSTR {parent=clavicle_l, loc=<5.9394 -0.077 -0.2179>, rot=<0.0107 0.4943 -0.0321 -0.8686>}
 shoulderBodyArmor_01_l_DSP {parent=shoulderBodyArmor_01_l_DSP_MSTR, loc=<0.2 0 0>, rot=<0 0 0 1>}
 shoulderBodyArmor_02_l_DSP {parent=shoulderBodyArmor_01_l_DSP, loc=<11.072 0 0>, rot=<0 0 0 1>}
 shoulderBodyArmor_back_01_l_DSP {parent=shoulderBodyArmor_01_l_DSP, loc=<-7.7789 -10.1329 0.5269>, rot=<0.0194 -0.6915 0.719 0.067>}
 shoulderBodyArmor_front_01_l_DSP {parent=shoulderBodyArmor_01_l_DSP, loc=<-7.6693 14.7607 0.0953>, rot=<0.0883 0.6678 0.7359 0.0685>}
 shoulderBodyArmor_upper_l_DSP {parent=shoulderBodyArmor_01_l_DSP, loc=<9.6513 0 4.0567>, rot=<-0.0282 0.2862 -0.0636 -0.9556>}
 shoulderBodyArmor_02_l_DSP2 {parent=shoulderBodyArmor_upper_l_DSP, loc=<6.4453 0 0>, rot=<0 0 0 1>}
 clavicle_r {parent=spine_05, loc=<4.3575 -0.1466 2.3223>, rot=<0.7289 0.0465 -0.6829 0.009>}
 upperarm_r {parent=clavicle_r, loc=<-15.5111 0 0>, rot=<0.0226 -0.3234 -0.0088 -0.9459>}
 lowerarm_r {parent=upperarm_r, loc=<-26.6515 0 0>, rot=<0.0391 -0.0099 -0.2453 -0.9686>}
 hand_r {parent=lowerarm_r, loc=<-22.3492 0 0>, rot=<-0.7081 -0.0507 0.0975 -0.6975>}
 index_metacarpal_r {parent=hand_r, loc=<-1.6534 -0.0181 2.9601>, rot=<0.0158 -0.0124 -0.0173 -0.9996>}
 index_01_r {parent=index_metacarpal_r, loc=<-8.3008 -0.0357 0.0524>, rot=<-0.0044 0.0485 0.0924 -0.9945>}
 index_02_r {parent=index_01_r, loc=<-3.3589 0.3003 -0.0453>, rot=<-0.002 0.0027 0.3225 -0.9466>}
 index_03_r {parent=index_02_r, loc=<-2.5578 0.1231 0.0251>, rot=<-0.002 0.0067 0.2558 -0.9667>}
 middle_metacarpal_r {parent=hand_r, loc=<-2.1856 0.1726 0.5984>, rot=<0.006 0.0168 -0.0329 -0.9993>}
 middle_01_r {parent=middle_metacarpal_r, loc=<-7.3273 0.2447 -0.0129>, rot=<-0.0208 0.0042 0.0858 -0.9961>}
 middle_02_r {parent=middle_01_r, loc=<-4.0142 -0.1411 -0.07>, rot=<-0.0217 0.0259 0.3512 -0.9357>}
 middle_03_r {parent=middle_02_r, loc=<-2.6074 -0.0534 -0.1312>, rot=<0.0083 0.0188 0.1708 -0.9851>}
 pinky_metacarpal_r {parent=hand_r, loc=<-1.7617 0.9753 -3.3814>, rot=<-0.0397 0.0814 -0.069 -0.9935>}
 pinky_01_r {parent=pinky_metacarpal_r, loc=<-7.1707 0.4832 0.1094>, rot=<-0.0654 -0.0035 0.2242 -0.9723>}
 pinky_02_r {parent=pinky_01_r, loc=<-2.8274 -0.0892 0.1043>, rot=<0.0037 0.0011 0.2854 -0.9584>}
 pinky_03_r {parent=pinky_02_r, loc=<-2.2045 -0.031 0.0078>, rot=<-0.0002 0.0098 0.1445 -0.9895>}
 ring_metacarpal_r {parent=hand_r, loc=<-2.2073 0.3627 -1.3638>, rot=<-0.0284 0.0722 -0.0458 -0.9959>}
 ring_01_r {parent=ring_metacarpal_r, loc=<-7.0297 0.6197 0.1446>, rot=<-0.0489 -0.0252 0.1892 -0.9804>}
 ring_02_r {parent=ring_01_r, loc=<-3.6171 -0.4648 0.1027>, rot=<0 0 0.2806 -0.9598>}
 ring_03_r {parent=ring_02_r, loc=<-2.4153 -0.1141 -0.0199>, rot=<0 0 0.1189 -0.9929>}
 thumb_01_r {parent=hand_r, loc=<-1.3119 0.3192 2.459>, rot=<0.6546 -0.4042 0.0664 -0.6354>}
 thumb_02_r {parent=thumb_01_r, loc=<-5.9335 0.0447 -0.0401>, rot=<-0.0037 0.0213 0.298 -0.9543>}
 thumb_03_r {parent=thumb_02_r, loc=<-2.2223 -0.3682 0.0907>, rot=<0.0118 -0.013 0.0066 -0.9998>}
 weapon_r {parent=hand_r, loc=<0.1975 3.4112 -0.3811>, rot=<0 0 0 -1>}
 lowerarm_twist_01_r {parent=lowerarm_r, loc=<-17.855 0 0>, rot=<0 0 0 1>}
 dyn_wristFlap_01_r_MSTR {parent=lowerarm_twist_01_r, loc=<-0.5561 3.5324 -2.8085>, rot=<0.1255 0.7303 -0.0562 -0.6692>}
 dyn_wristFlap_01_r {parent=dyn_wristFlap_01_r_MSTR, loc=<-0.2 0 0>, rot=<0 0 0 1>}
 lowerarm_twist_02_r {parent=lowerarm_r, loc=<-9.6199 0 0>, rot=<0 0 0 1>}
 upperarm_twist_01_r {parent=upperarm_r, loc=<-0.0041 0 0>, rot=<0 0 0 1>}
 shoulderpad_01_r_DSP {parent=upperarm_twist_01_r, loc=<-1.5186 0 0.0002>, rot=<0 0 0 1>}
 shoulderpad_02_r_DSP {parent=shoulderpad_01_r_DSP, loc=<-0.0001 0 -5.864>, rot=<0 0 0 1>}
 upperarm_jacket_01_r_DSP {parent=upperarm_twist_01_r, loc=<-0.0025 -0.0112 0.002>, rot=<0 0 0 1>}
 upperarm_twist_02_r {parent=upperarm_r, loc=<-13.7942 0 0>, rot=<0 0 0 1>}
 dyn_parent_shldr_pad_r {parent=upperarm_r, loc=<-4.8685 0 0>, rot=<-0.0001 0.0904 0.0046 -0.9959>}
 dyn_shldr_pad_base_r {parent=dyn_parent_shldr_pad_r, loc=<0 0 0>, rot=<0 0 0 1>}
 dyn_shldr_pad_bot_r {parent=dyn_shldr_pad_base_r, loc=<-10.7527 0.1309 -2.35>, rot=<-0.0264 -0.7092 -0.7023 0.0556>}
 dyn_shldr_pad_top_r {parent=dyn_shldr_pad_bot_r, loc=<-20.0906 -1.3738 0>, rot=<0 0 0 1>}
 dyn_shldr_frt_r {parent=upperarm_r, loc=<-4.7484 -6.0025 -0.9203>, rot=<-0.0001 0.0904 0.0046 -0.9959>}
 dyn_shldr_bck_r {parent=upperarm_r, loc=<-4.8582 7.2332 -1.0677>, rot=<-0.0001 0.0904 0.0046 -0.9959>}
 dyn_shldr_aim_r {parent=upperarm_r, loc=<-21.163 0 0>, rot=<0 0 0 1>}
 neck_01 {parent=spine_05, loc=<11.3138 0.1529 0>, rot=<0 0 -0.2003 -0.9797>}
 neck_02 {parent=neck_01, loc=<4.5627 0 0>, rot=<0 0 0.0504 -0.9987>}
 head {parent=neck_02, loc=<5.3914 0 0>, rot=<0 0 0.115 -0.9934>}
 dyn_collar1_01_l_MSTR {parent=neck_01, loc=<0.0853 10.2816 -6.3453>, rot=<-0.2362 -0.0246 -0.0432 -0.9704>}
 dyn_collar1_01_l {parent=dyn_collar1_01_l_MSTR, loc=<0.2 0 0>, rot=<0 0 0 1>}
 dyn_collar1_02_l {parent=dyn_collar1_01_l, loc=<3.2246 0 0>, rot=<0 0 0 1>}
 dyn_collar2_01_l_MSTR {parent=neck_01, loc=<1.4288 2.2992 -9.7169>, rot=<-0.1164 -0.0619 0.0331 -0.9907>}
 dyn_collar2_01_l {parent=dyn_collar2_01_l_MSTR, loc=<0.2 0 0>, rot=<0 0 0 1>}
 dyn_collar2_02_l {parent=dyn_collar2_01_l, loc=<3.2246 0 0>, rot=<0 0 0 1>}
 dyn_collar3_01_l_MSTR {parent=neck_01, loc=<1.0757 -6.1864 -9.0803>, rot=<0.3488 -0.044 0.0721 -0.9334>}
 dyn_collar3_01_l {parent=dyn_collar3_01_l_MSTR, loc=<0.2 0 0>, rot=<0 0 0 1>}
 dyn_collar3_02_l {parent=dyn_collar3_01_l, loc=<3.2246 0 0>, rot=<0 0 0 1>}
 dyn_collar_back_01_MSTR {parent=neck_01, loc=<-0.7179 -9.6814 0>, rot=<0.6954 -0.1282 0.1282 -0.6954>}
 dyn_collar_back_01 {parent=dyn_collar_back_01_MSTR, loc=<0.2 0 0>, rot=<0 0 0 1>}
 dyn_collar_back_02 {parent=dyn_collar_back_01, loc=<4.077 0 0>, rot=<0 0 0 1>}
 dyn_collar3_01_r_MSTR {parent=neck_01, loc=<1.076 -6.1865 9.0803>, rot=<-0.044 -0.3488 0.9334 0.0721>}
 dyn_collar3_01_r {parent=dyn_collar3_01_r_MSTR, loc=<-0.1996 -0.0001 0>, rot=<0 0 0 1>}
 dyn_collar3_02_r {parent=dyn_collar3_01_r, loc=<-3.2252 0.0001 0>, rot=<0 0 0 1>}
 dyn_collar2_01_r_MSTR {parent=neck_01, loc=<1.4291 2.2991 9.7169>, rot=<-0.0619 0.1164 0.9907 0.0331>}
 dyn_collar2_01_r {parent=dyn_collar2_01_r_MSTR, loc=<-0.2 0 0>, rot=<0 0 0 1>}
 dyn_collar2_02_r {parent=dyn_collar2_01_r, loc=<-3.2245 0 0>, rot=<0 0 0 -1>}
 dyn_collar1_01_r_MSTR {parent=neck_01, loc=<0.0848 10.2818 6.3453>, rot=<-0.0246 0.2362 0.9704 -0.0432>}
 dyn_collar1_01_r {parent=dyn_collar1_01_r_MSTR, loc=<-0.2003 0.0001 -0.0001>, rot=<0 0 0 1>}
 dyn_collar1_02_r {parent=dyn_collar1_01_r, loc=<-3.2247 0 0>, rot=<0 0 0 1>}
 dyn_jacket_upper_01_l_MSTR {parent=spine_05, loc=<-0.2456 11.9078 -8.1871>, rot=<-0.2615 0.4079 -0.5902 -0.6457>}
 dyn_jacket_upper_01_l {parent=dyn_jacket_upper_01_l_MSTR, loc=<0.2 0 0>, rot=<0 0 0 1>}
 dyn_jacket_upper_02_l {parent=dyn_jacket_upper_01_l, loc=<4.1993 0 0>, rot=<0 0 0 1>}
 dyn_jacket_upper_01_r_MSTR {parent=spine_05, loc=<-0.246 11.9077 8.1871>, rot=<0.4079 0.2615 0.6457 -0.5902>}
 dyn_jacket_upper_01_r {parent=dyn_jacket_upper_01_r_MSTR, loc=<-0.2 0.0001 0>, rot=<0 0 0 1>}
 dyn_jacket_upper_02_r {parent=dyn_jacket_upper_01_r, loc=<-4.1993 0.0001 0>, rot=<0 0 0 1>}
 dyn_collar_front_01_MSTR {parent=spine_05, loc=<2.5896 11.1138 0>, rot=<0.6775 0.2026 -0.2026 -0.6775>}
 dyn_collar_front_01 {parent=dyn_collar_front_01_MSTR, loc=<0.2 0 0>, rot=<0 0 0 1>}
 dyn_hoodie_01_MSTR {parent=spine_05, loc=<11.7598 -12.0862 0>, rot=<0.3392 -0.6204 0.6204 -0.3392>}
 dyn_hoodie_01 {parent=dyn_hoodie_01_MSTR, loc=<0.2 0 0>, rot=<0 0 0 1>}
 strap_back_01_DSP_MSTR {parent=spine_03, loc=<9.4112 -13.0156 4.4026>, rot=<0.2816 -0.6195 0.7003 0.2158>}
 strap_back_01_DSP {parent=strap_back_01_DSP_MSTR, loc=<0 0 0>, rot=<0 0 0 1>}
 strap_front_01_DSP_MSTR {parent=spine_03, loc=<7.1595 13.9327 2.0673>, rot=<-0.251 -0.6803 -0.6632 0.1855>}
 strap_front_01_DSP {parent=strap_front_01_DSP_MSTR, loc=<0 0 0>, rot=<0 0 0 1>}
 dyn_jacket_mid_01_l_MSTR {parent=spine_03, loc=<6.1842 11.8511 -8.5603>, rot=<-0.3642 0.342 -0.6356 -0.5885>}
 dyn_jacket_mid_01_l {parent=dyn_jacket_mid_01_l_MSTR, loc=<0.2 -0.0004 0>, rot=<0 0 0 1>}
 dyn_jacket_mid_02_l {parent=dyn_jacket_mid_01_l, loc=<4.1993 -0.0003 0>, rot=<0 0 0 1>}
 dyn_jacket_front_01_l_MSTR {parent=spine_03, loc=<2.2367 12.8169 -8.9815>, rot=<0.0145 -0.707 -0.707 0.0145>}
 dyn_jacket_front_01_l {parent=dyn_jacket_front_01_l_MSTR, loc=<0.2 0 0>, rot=<0 0 0 1>}
 dyn_jacket_front_02_l {parent=dyn_jacket_front_01_l, loc=<6.5514 0 0>, rot=<0 0 0 1>}
 dyn_jacket_front_03_l {parent=dyn_jacket_front_02_l, loc=<6.5514 0 0>, rot=<0 0 0 1>}
 dyn_jacket_front_04_l {parent=dyn_jacket_front_03_l, loc=<6.5183 0 0.0001>, rot=<0 0 0 1>}
 dyn_jacket_front_01_r_MSTR {parent=spine_03, loc=<2.2372 12.8169 8.9815>, rot=<-0.707 -0.0145 -0.0145 -0.707>}
 dyn_jacket_front_01_r {parent=dyn_jacket_front_01_r_MSTR, loc=<-0.2005 0 0>, rot=<0 0 0 1>}
 dyn_jacket_front_02_r {parent=dyn_jacket_front_01_r, loc=<-6.5514 0 0>, rot=<0 0 0 1>}
 dyn_jacket_front_03_r {parent=dyn_jacket_front_02_r, loc=<-6.5514 0 0>, rot=<0 0 0 1>}
 dyn_jacket_front_04_r {parent=dyn_jacket_front_03_r, loc=<-6.5183 0 -0.0001>, rot=<0 0 0 1>}
 dyn_jacket_mid_01_r_MSTR {parent=spine_03, loc=<6.1844 11.8511 8.5603>, rot=<0.342 0.3642 0.5885 -0.6356>}
 dyn_jacket_mid_01_r {parent=dyn_jacket_mid_01_r_MSTR, loc=<-0.2 0.0004 0>, rot=<0 0 0 1>}
 dyn_jacket_mid_02_r {parent=dyn_jacket_mid_01_r, loc=<-4.1993 0.0003 0>, rot=<0 0 0 1>}
 dyn_strap {parent=spine_03, loc=<14.5794 -11.7522 -0.8403>, rot=<0.0887 -0.3672 -0.0149 -0.9258>}
 dyn_jacket1_01_l_MSTR {parent=spine_02, loc=<4.1931 5.3514 -15.216>, rot=<0.036 -0.9851 -0.1678 -0.0126>}
 dyn_jacket1_01_l {parent=dyn_jacket1_01_l_MSTR, loc=<0.2 0 0>, rot=<0 0 0 1>}
 dyn_jacket1_02_l {parent=dyn_jacket1_01_l, loc=<5.3304 0 0>, rot=<0 0 0 1>}
 dyn_jacket1_01_r_MSTR {parent=spine_02, loc=<4.193 5.3514 15.216>, rot=<0.9851 0.036 -0.0126 0.1678>}
 dyn_jacket1_01_r {parent=dyn_jacket1_01_r_MSTR, loc=<-0.2002 0 0>, rot=<0 0 0 1>}
 dyn_jacket1_02_r {parent=dyn_jacket1_01_r, loc=<-5.3301 0 0>, rot=<0 0 0 1>}
 dyn_jacket_back_01_MSTR {parent=spine_02, loc=<4.4554 -10.642 0>, rot=<0.0707 -0.7036 0.7036 -0.0707>}
 dyn_jacket_back_01 {parent=dyn_jacket_back_01_MSTR, loc=<0.2 0 0>, rot=<0 0 0 1>}
 dyn_jacket_back_02 {parent=dyn_jacket_back_01, loc=<8.1119 0 0>, rot=<0 0 0 1>}
 dyn_jacket2_01_l_MSTR {parent=spine_02, loc=<4.4103 -9.715 -10.109>, rot=<0.0723 -0.7034 0.7034 -0.0723>}
 dyn_jacket2_01_l {parent=dyn_jacket2_01_l_MSTR, loc=<0.1996 0 0.0001>, rot=<0 0 0 1>}
 dyn_jacket2_02_l {parent=dyn_jacket2_01_l, loc=<5.4576 0 -0.0001>, rot=<0 0 0 1>}
 dyn_jacket2_01_r_MSTR {parent=spine_02, loc=<4.4103 -9.715 10.109>, rot=<0.7034 0.0723 -0.0723 -0.7034>}
 dyn_jacket2_01_r {parent=dyn_jacket2_01_r_MSTR, loc=<-0.1996 0 -0.0001>, rot=<0 0 0 1>}
 dyn_jacket2_02_r {parent=dyn_jacket2_01_r, loc=<-5.4576 0 0.0001>, rot=<0 0 0 1>}
 dyn_shirt_front_01_MSTR {parent=spine_01, loc=<4.872 10.3951 0>, rot=<-0.0128 -0.707 -0.707 -0.0128>}
 dyn_shirt_front_01 {parent=dyn_shirt_front_01_MSTR, loc=<0.2 0 0>, rot=<0 0 0 1>}
 dyn_shirt_front_02 {parent=dyn_shirt_front_01, loc=<8.1099 0 0>, rot=<0 0 0 1>}
 dyn_shirt_front_03 {parent=dyn_shirt_front_02, loc=<7.9617 0 0>, rot=<0 0 0 1>}
 dyn_shirt_front_04 {parent=dyn_shirt_front_03, loc=<7.9617 0 0>, rot=<0 0 0 1>}
 dyn_shirt_front_01_r_MSTR {parent=spine_01, loc=<4.4723 10.4096 6.951>, rot=<-0.0128 -0.707 -0.707 -0.0128>}
 dyn_shirt_front_01_r {parent=dyn_shirt_front_01_r_MSTR, loc=<-0.2 0 0>, rot=<0 0 0 1>}
 dyn_shirt_front_02_r {parent=dyn_shirt_front_01_r, loc=<8.1099 0 0>, rot=<0 0 0 1>}
 dyn_shirt_front_03_r {parent=dyn_shirt_front_02_r, loc=<7.9617 0 0>, rot=<0 0 0 1>}
 dyn_shirt_front_04_r {parent=dyn_shirt_front_03_r, loc=<7.9617 0 0>, rot=<0 0 0 1>}
 dyn_shirt_front_01_l_MSTR {parent=spine_01, loc=<4.872 10.3951 -6.9508>, rot=<-0.0128 -0.707 -0.707 -0.0128>}
 dyn_shirt_front_01_l {parent=dyn_shirt_front_01_l_MSTR, loc=<0.2 0 0>, rot=<0 0 0 1>}
 dyn_shirt_front_02_l {parent=dyn_shirt_front_01_l, loc=<8.1099 0 0>, rot=<0 0 0 1>}
 dyn_shirt_front_03_l {parent=dyn_shirt_front_02_l, loc=<7.9617 0 0>, rot=<0 0 0 1>}
 dyn_shirt_front_04_l {parent=dyn_shirt_front_03_l, loc=<7.9617 0 0>, rot=<0 0 0 1>}
 thigh_l {parent=pelvis, loc=<-2.5693 -0.0109 -7.7501>, rot=<0.0012 0.0572 0.0155 -0.9982>}
 calf_l {parent=thigh_l, loc=<-39.0484 0 0>, rot=<0.0006 -0.0161 -0.0391 -0.9991>}
 calf_twist_01_l {parent=calf_l, loc=<-35.813 -0.0967 0.0397>, rot=<0.0237 0.0095 -0.0035 -0.9997>}
 calf_twist_02_l {parent=calf_l, loc=<-18.8064 -0.0967 0.0397>, rot=<0.0237 0.0095 -0.0035 -0.9997>}
 foot_l {parent=calf_l, loc=<-42.561 0.1077 0.4292>, rot=<0.0418 -0.0455 0.0343 -0.9975>}
 ball_l {parent=foot_l, loc=<-6.1599 13.8264 0.6806>, rot=<0.0001 -0.0001 -0.7063 -0.7079>}
 shoe_01_l_DSP {parent=foot_l, loc=<0 0 0>, rot=<0 0 -0.3183 -0.948>}
 shoe_02_l_DSP {parent=shoe_01_l_DSP, loc=<8.7469 0 0>, rot=<0 0 0 1>}
 thigh_twist_01_l {parent=thigh_l, loc=<-21.6303 0 0>, rot=<0.0013 0 0 -1>}
 thigh_front_01_l_DSP {parent=thigh_twist_01_l, loc=<7.5693 9.3773 0.0238>, rot=<-0.0013 0 0 -1>}
 thigh_back_01_l_DSP {parent=thigh_twist_01_l, loc=<7.4857 -7.8077 -0.0223>, rot=<-0.0013 0 0 -1>}
 thigh_col_02_l {parent=thigh_l, loc=<0 -0.779 0>, rot=<0 0 0 1>}
 thigh_col_01_l {parent=thigh_l, loc=<0 0.7788 0>, rot=<0 0 0 1>}
 thigh_r {parent=pelvis, loc=<-2.5691 -0.0109 7.7501>, rot=<0.0572 -0.0012 0.9982 0.0155>}
 calf_r {parent=thigh_r, loc=<39.0487 0 0>, rot=<0.0006 -0.0161 -0.0391 -0.9991>}
 calf_twist_01_r {parent=calf_r, loc=<35.8132 0.0967 -0.0397>, rot=<0.0237 0.0095 -0.0035 -0.9997>}
 calf_twist_02_r {parent=calf_r, loc=<18.8064 0.0967 -0.0397>, rot=<0.0237 0.0095 -0.0035 -0.9997>}
 foot_r {parent=calf_r, loc=<42.5612 -0.1077 -0.4292>, rot=<0.0418 -0.0455 0.0343 -0.9975>}
 ball_r {parent=foot_r, loc=<6.1598 -13.8264 -0.6806>, rot=<0.0001 -0.0001 -0.7063 -0.7079>}
 shoe_01_r_DSP {parent=foot_r, loc=<-0.0002 0 -0.0001>, rot=<0 0 -0.3183 -0.948>}
 shoe_02_r_DSP {parent=shoe_01_r_DSP, loc=<-8.747 0 0.0001>, rot=<0 0 0 1>}
 thigh_twist_01_r {parent=thigh_r, loc=<21.6303 0 0>, rot=<0.0013 0 0 -1>}
 dyn_legpouch_r_01_MSTR {parent=thigh_twist_01_r, loc=<0.1869 -0.5634 7.9083>, rot=<0 0.6325 0 -0.7746>}
 dyn_legpouch_r_01 {parent=dyn_legpouch_r_01_MSTR, loc=<0.2 0 0>, rot=<0 0 0 1>}
 thigh_front_01_r_DSP {parent=thigh_twist_01_r, loc=<-7.5691 -9.3773 -0.0238>, rot=<-0.0013 0 0 -1>}
 thigh_back_01_r_DSP {parent=thigh_twist_01_r, loc=<-7.4855 7.8077 0.0223>, rot=<-0.0013 0 0 -1>}
 thigh_col_02_r {parent=thigh_r, loc=<0.0002 0.779 0>, rot=<0 0 0 1>}
 thigh_col_01_r {parent=thigh_r, loc=<0.0002 -0.7788 0>, rot=<0 0 0 1>}
 dyn_jacket_back_01_l_MSTR {parent=pelvis, loc=<3.4777 -10.3993 -5.8877>, rot=<0.0315 -0.7064 0.7064 -0.0315>}
 dyn_jacket_back_01_l {parent=dyn_jacket_back_01_l_MSTR, loc=<0.2 0 0>, rot=<0 0 0 1>}
 dyn_jacket_back_02_l {parent=dyn_jacket_back_01_l, loc=<7.9617 0 0>, rot=<0 0 0 1>}
 dyn_jacket_back_03_l {parent=dyn_jacket_back_02_l, loc=<7.9617 0 0>, rot=<0 0 0 1>}
 dyn_jacket_back_01_r_MSTR {parent=pelvis, loc=<3.4776 -10.3993 5.8877>, rot=<0.7064 0.0315 -0.0315 -0.7064>}
 dyn_jacket_back_01_r {parent=dyn_jacket_back_01_r_MSTR, loc=<-0.2 0 0>, rot=<0 0 0 1>}
 dyn_jacket_back_02_r {parent=dyn_jacket_back_01_r, loc=<-7.9616 0 0>, rot=<0 0 0 1>}
 dyn_jacket_back_03_r {parent=dyn_jacket_back_02_r, loc=<-7.9617 0 0.0001>, rot=<0 0 0 1>}
 pelvis_col_01 {parent=pelvis, loc=<0 0.7514 0>, rot=<0 0 0 1>}
 pelvis_col_02 {parent=pelvis, loc=<0 -0.751 0>, rot=<0 0 0 1>}
 pelvis_back_01_DSP {parent=pelvis, loc=<0 -0.2178 0>, rot=<0 0 0 1>}
 ik_foot_root {parent=root, loc=<0 0 0>, rot=<0 0 0 1>}
 ik_foot_l {parent=ik_foot_root, loc=<15.2608 -4.4176 9.4433>, rot=<0.0357 0.7045 0.0284 -0.7082>}
 ik_foot_r {parent=ik_foot_root, loc=<-15.2609 -4.4176 9.4431>, rot=<0.7082 0.0284 -0.7045 0.0357>}
 ik_hand_root {parent=root, loc=<0 0 0>, rot=<0 0 0 1>}
 ik_hand_gun {parent=ik_hand_root, loc=<-51.9729 4.0884 102.6935>, rot=<0.6084 0.1793 0.4613 -0.6205>}
 ik_hand_l {parent=ik_hand_gun, loc=<53.0286 82.1782 35.2051>, rot=<-0.5102 -0.7906 -0.3387 0>}
 ik_hand_r {parent=ik_hand_gun, loc=<0 0 0>, rot=<0 0 0 1>}
]
