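# Physics asset for the back jacket of the male medium Rust Remix player cosmetic, set up for secondary motion of the cloth flaps.
# Three two-link capsule chains (left, right, center back) each hang from a tiny box MSTR anchor body through constraints sharing identical swing and twist limits.
# Capsules on the pelvis, spine and dedicated thigh collision bones approximate the body as unconstrained collision shapes for the jacket.
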PhysicsAsset /BRCosmetics/Characters/Player/Male/Medium/Bodies/M_MED_Rust_Remix/Meshes/M_MED_Rust_Remix_jacketBack_Physics
bodies {
 dyn_jacket2_01_l [{shapeType=Capsule, length=3.7949, radius=6.1856, shapeName=dyn_jacket2_01_l_capsule, center=<2.9868 0 3>, rotation=<rot 0 0 89.9992>}]
 dyn_jacket2_01_l_MSTR [{shapeType=Box, extentX=0.01, extentY=0.01, extentZ=0.01}]
 dyn_jacket2_01_r [{shapeType=Capsule, length=3.7949, radius=6.1856, shapeName=dyn_jacket2_01_r_capsule, center=<-2.9868 0 -3>, rotation=<rot 0 0 89.9994>}]
 dyn_jacket2_01_r_MSTR [{shapeType=Box, extentX=0.01, extentY=0.01, extentZ=0.01}]
 dyn_jacket2_02_l [{shapeType=Capsule, length=3.7949, radius=6.1856, shapeName=dyn_jacket2_02_l_capsule, center=<6.6795 0 3>, rotation=<rot 0 0 89.9992>}]
 dyn_jacket2_02_r [{shapeType=Capsule, length=3.7949, radius=6.1856, shapeName=dyn_jacket2_02_r_capsule, center=<-6.6795 0 -3>, rotation=<rot 0 0 89.9994>}]
 dyn_jacket_back_01 [{shapeType=Capsule, length=3.7949, radius=6.1856, shapeName=dyn_jacket_back_01_capsule, center=<6.6795 0 3>, rotation=<rot 0 0 90.0003>}]
 dyn_jacket_back_01_MSTR [{shapeType=Box, extentX=0.01, extentY=0.01, extentZ=0.01}]
 dyn_jacket_back_02 [{shapeType=Capsule, length=3.7949, radius=6.1856, shapeName=dyn_jacket_back_02_capsule, center=<6.6795 0 3>, rotation=<rot 0 0 90.0003>}]
 pelvis [{shapeType=Capsule, length=17.25, radius=9.75}, {shapeType=Capsule, length=17, radius=9, center=<6.3679 0.9589 0.0001>}]
 spine_01 [{shapeType=Capsule, length=11.5, radius=10, center=<0 -0.3711 0>}]
 spine_02 [{shapeType=Capsule, length=7.5, radius=10.5, center=<0 1.29 0>}]
 spine_03 [{shapeType=Capsule, length=11, radius=9.5, center=<0 0.5812 0>}]
 thigh_col_01_l [{shapeType=Capsule, length=31.75, radius=11.5, center=<-11.461 -0.9226 -0.0003>, rotation=<rot -90 0.0396 -0.0396>}]
 thigh_col_01_r [{shapeType=Capsule, length=31.75, radius=11.5, center=<11.461 0.9226 0.0003>, rotation=<rot -90 -0.0264 -179.9736>}]
}
constraints [
 {bone1Name=dyn_jacket_back_01, bone2Name=dyn_jacket_back_01_MSTR, swing1LimitDegrees=15, swing2LimitDegrees=30, twistLimitDegrees=15}
 {bone1Name=dyn_jacket_back_02, bone2Name=dyn_jacket_back_01, swing1LimitDegrees=15, swing2LimitDegrees=30, twistLimitDegrees=15}
 {bone1Name=dyn_jacket2_01_l, bone2Name=dyn_jacket2_01_l_MSTR, swing1LimitDegrees=15, swing2LimitDegrees=30, twistLimitDegrees=15}
 {bone1Name=dyn_jacket2_02_l, bone2Name=dyn_jacket2_01_l, swing1LimitDegrees=15, swing2LimitDegrees=30, twistLimitDegrees=15}
 {bone1Name=dyn_jacket2_01_r, bone2Name=dyn_jacket2_01_r_MSTR, swing1LimitDegrees=15, swing2LimitDegrees=30, twistLimitDegrees=15}
 {bone1Name=dyn_jacket2_02_r, bone2Name=dyn_jacket2_01_r, swing1LimitDegrees=15, swing2LimitDegrees=30, twistLimitDegrees=15}
]
