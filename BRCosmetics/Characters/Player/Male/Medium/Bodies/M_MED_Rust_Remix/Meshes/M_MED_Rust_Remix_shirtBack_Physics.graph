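# Physics asset for the back shirt/jacket flaps of the male medium Rust Remix player cosmetic body.
# Left and right chains of three capsule bodies each are constrained in series from a tiny box body on that side's MSTR bone.
# Tight twist and swing1 limits with a wider swing2 make each joint act mostly as a limited hinge.
# Unconstrained pelvis and thigh capsules provide body shapes, likely as colliders for the flaps.
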PhysicsAsset /BRCosmetics/Characters/Player/Male/Medium/Bodies/M_MED_Rust_Remix/Meshes/M_MED_Rust_Remix_shirtBack_Physics
bodies {
 dyn_jacket_back_01_l [{shapeType=Capsule, length=6.7729, radius=5.9271, shapeName=dyn_jacket_back_01_l_capsule, center=<4.0825 0 3>, rotation=<rot 0 0 90>}]
 dyn_jacket_back_01_l_MSTR [{shapeType=Box, extentX=0.01, extentY=0.01, extentZ=0.01}]
 dyn_jacket_back_01_r [{shapeType=Capsule, length=7.2388, radius=5.4612, shapeName=dyn_jacket_back_01_r_capsule, center=<-4.0825 0 -4>, rotation=<rot 0 0 90>}]
 dyn_jacket_back_01_r_MSTR [{shapeType=Box, extentX=0.01, extentY=0.01, extentZ=0.01}]
 dyn_jacket_back_02_l [{shapeType=Capsule, length=6.7729, radius=5.9271, shapeName=dyn_jacket_back_02_l_capsule, center=<4.0825 0 3>, rotation=<rot 0 0 90>}]
 dyn_jacket_back_02_r [{shapeType=Capsule, length=7.4921, radius=5.2079, shapeName=dyn_jacket_back_02_r_capsule, center=<-4.0825 0 -3>, rotation=<rot 0 0 90>}]
 dyn_jacket_back_03_l [{shapeType=Capsule, length=6.7891, radius=5.9271, shapeName=dyn_jacket_back_03_l_capsule, center=<4.0825 0 3>, rotation=<rot 0 0 90>}]
 dyn_jacket_back_03_r [{shapeType=Capsule, length=7.3176, radius=5.3824, shapeName=dyn_jacket_back_03_r_capsule, center=<-4.0825 0 -3>, rotation=<rot 0 0 90>}]
 pelvis [{shapeType=Capsule, length=17.25, radius=8.875}, {shapeType=Capsule, length=17, radius=9, center=<6.3679 0.9589 0.0001>}]
 thigh_l [{shapeType=Capsule, length=33.681, radius=9.7073, center=<-20.198 0.0002 0>, rotation=<rot 90 0 0>}]
 thigh_r [{shapeType=Capsule, length=33.681, radius=9.7073, center=<20.198 -0.0002 0>, rotation=<rot 90 5.0512 -174.9487>}]
}
constraints [
 {bone1Name=dyn_jacket_back_01_l, bone2Name=dyn_jacket_back_01_l_MSTR, swing1LimitDegrees=1, swing2LimitDegrees=30, twistLimitDegrees=1}
 {bone1Name=dyn_jacket_back_02_l, bone2Name=dyn_jacket_back_01_l, swing1LimitDegrees=1, swing2LimitDegrees=30, twistLimitDegrees=1}
 {bone1Name=dyn_jacket_back_03_l, bone2Name=dyn_jacket_back_02_l, swing1LimitDegrees=1, swing2LimitDegrees=30, twistLimitDegrees=1}
 {bone1Name=dyn_jacket_back_01_r, bone2Name=dyn_jacket_back_01_r_MSTR, swing1LimitDegrees=1, swing2LimitDegrees=30, twistLimitDegrees=1}
 {bone1Name=dyn_jacket_back_02_r, bone2Name=dyn_jacket_back_01_r, swing1LimitDegrees=1, swing2LimitDegrees=30, twistLimitDegrees=1}
 {bone1Name=dyn_jacket_back_03_r, bone2Name=dyn_jacket_back_02_r, swing1LimitDegrees=1, swing2LimitDegrees=30, twistLimitDegrees=1}
]
