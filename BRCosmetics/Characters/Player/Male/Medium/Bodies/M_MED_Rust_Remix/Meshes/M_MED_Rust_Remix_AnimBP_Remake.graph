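# Post-process style Anim Blueprint for the Rust Remix cosmetic body, adding secondary motion to its jacket, shirt, collar, wrist flaps, leg pouch and shoes.
# Copied mesh pose runs through Control Rig, constraints, AnimDynamics and pelvis-space RigidBody sims into preSimCache.
# lod2Cache layers offsets and constraints on that cache, chosen by isMobileLOD before final corrective bone tweaks.
# EventGraph derives corrective alphas from limb and head angles, character state and frame rate.
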AnimBlueprint /BRCosmetics/Characters/Player/Male/Medium/Bodies/M_MED_Rust_Remix/Meshes/M_MED_Rust_Remix_AnimBP_Remake
target_skeleton=&/M_MED_Rust_Remix_Skeleton
graphs [
 AnimGraph {nodes [
   n1 = Use cached pose 'preSimCache'()
   n2 = Use cached pose 'lod2Cache'()
   n3 = Get isMobileLOD() -> (isMobileLOD:bool)
   n4 = Get armSwingDelta_r() -> (armSwingDelta_r:double)
   n5 = "Transform (Modify) Bone - Bone: upperarm_jacket_01_r_DSP"(
    ComponentPose=Local To Component(LocalPose=Blend Poses by bool(BlendPose_0=@n2, BlendPose_1=@n1, bActiveValue=@n3) {blendType=Linear})
    Alpha=@n4
    Translation="0,0,0"
    Rotation="0,0,0"
    Scale="1.5,1.5,1.5"
   ) {
    rotationMode=BMM_Additive
    rotationSpace=BCS_BoneSpace
    scaleMode=BMM_Additive
    scaleSpace=BCS_BoneSpace
    translationMode=BMM_Additive
    translationSpace=BCS_BoneSpace
    alphaBoolBlend {blendOption=Linear}
    boneToModify {boneName=upperarm_jacket_01_r_DSP}
    scale=<1.5 1.5 1.5>
   }
   n6 = "Transform (Modify) Bone - Bone: upperarm_twist_02_r"(ComponentPose=@n5, Alpha=@n4, Translation="0,0,0", Rotation="0,0,0", Scale="1.25,1.25,1.25") {
    rotationMode=BMM_Additive
    rotationSpace=BCS_BoneSpace
    scaleMode=BMM_Additive
    scaleSpace=BCS_BoneSpace
    translationMode=BMM_Additive
    translationSpace=BCS_BoneSpace
    alphaBoolBlend {blendOption=Linear}
    boneToModify {boneName=upperarm_twist_02_r}
    scale=<1.25 1.25 1.25>
   }
   n7 = "Transform (Modify) Bone - Bone: upperarm_jacket_01_l_DSP"(ComponentPose=@n6, Alpha=@n4, Translation="0,0,0", Rotation="0,0,0", Scale="1.5,1.5,1.5") {
    rotationMode=BMM_Additive
    rotationSpace=BCS_BoneSpace
    scaleMode=BMM_Additive
    scaleSpace=BCS_BoneSpace
    translationMode=BMM_Additive
    translationSpace=BCS_BoneSpace
    alphaBoolBlend {blendOption=Linear}
    boneToModify {boneName=upperarm_jacket_01_l_DSP}
    scale=<1.5 1.5 1.5>
   }
   n8 = "Transform (Modify) Bone - Bone: upperarm_twist_02_l"(ComponentPose=@n7, Alpha=@n4, Translation="0,0,0", Rotation="0,0,0", Scale="1.25,1.25,1.25") {
    rotationMode=BMM_Additive
    rotationSpace=BCS_BoneSpace
    scaleMode=BMM_Additive
    scaleSpace=BCS_BoneSpace
    translationMode=BMM_Additive
    translationSpace=BCS_BoneSpace
    alphaBoolBlend {blendOption=Linear}
    boneToModify {boneName=upperarm_twist_02_l}
    scale=<1.25 1.25 1.25>
   }
   n9 = "Constraint - shoe_01_l_DSP (P)"(ComponentPose=@n8, Alpha=1, ConstraintWeights_0=1) {
    alphaBoolBlend {blendOption=Linear}
    boneToModify {boneName=shoe_01_l_DSP}
    constraintSetup [{targetBone {boneName=calf_twist_02_l}}]
    constraintWeights [1]
   }
   n10 = Flap Node - Bone: shoe_01_l_DSP!(ComponentPose:Component Space Pose Link Structure=@n9, Alpha:float=1) -> (Pose:Component Space Pose Link Structure) {
    is=title FortAnimGraphNode_Flap
    bUseClamp=true
    targetClampMax=1
    targetComponent=RotationZ
    alphaBoolBlend {blendOption=Linear}
    sourceDrives [{
      bUseRange=false
      drivingCurve=null
      multiplier=0.8
      rangeMax=1
      rangeMin=-1
      remappedMax=1
      remappedMin=0
      sourceComponent=RotationZ
      useQuaternion=true
      sourceBone {boneName=foot_l}
     }]
    targetBone {boneName=shoe_01_l_DSP}
   }
   n11 = "Constraint - shoe_01_r_DSP (P)"(ComponentPose=@n10, Alpha=1, ConstraintWeights_0=1) {
    alphaBoolBlend {blendOption=Linear}
    boneToModify {boneName=shoe_01_r_DSP}
    constraintSetup [{targetBone {boneName=calf_twist_02_r}}]
    constraintWeights [1]
   }
   n12 = Flap Node - Bone: shoe_01_r_DSP!(ComponentPose:Component Space Pose Link Structure=@n11, Alpha:float=1) -> (Pose:Component Space Pose Link Structure) {
    is=title FortAnimGraphNode_Flap
    bUseClamp=true
    targetClampMax=1
    targetComponent=RotationZ
    alphaBoolBlend {blendOption=Linear}
    sourceDrives [{
      bUseRange=false
      drivingCurve=null
      multiplier=0.8
      rangeMax=1
      rangeMin=-1
      remappedMax=1
      remappedMin=0
      sourceComponent=RotationZ
      useQuaternion=true
      sourceBone {boneName=foot_r}
     }]
    targetBone {boneName=shoe_01_r_DSP}
   }
   n13 = Get headSwingAngleAlpha() -> (headSwingAngleAlpha:double)
   n14 = "Transform (Modify) Bone - Bone: dyn_collar3_02_r"(ComponentPose=@n12, Alpha=@n13, Translation="0,0,0", Rotation="-15,0,0", Scale="1,1,1") {
    rotationMode=BMM_Additive
    rotationSpace=BCS_BoneSpace
    scaleMode=BMM_Additive
    scaleSpace=BCS_BoneSpace
    translationMode=BMM_Additive
    translationSpace=BCS_BoneSpace
    alphaBoolBlend {blendOption=Linear}
    boneToModify {boneName=dyn_collar3_02_r}
    rotation {pitch=-15}
   }
   n15 = "Transform (Modify) Bone - Bone: dyn_collar_back_02"(ComponentPose=@n14, Alpha=@n13, Translation="0,0,0", Rotation="-15,0,0", Scale="1,1,1") {
    rotationMode=BMM_Additive
    rotationSpace=BCS_BoneSpace
    scaleMode=BMM_Additive
    scaleSpace=BCS_BoneSpace
    translationMode=BMM_Additive
    translationSpace=BCS_BoneSpace
    alphaBoolBlend {blendOption=Linear}
    boneToModify {boneName=dyn_collar_back_02}
    rotation {pitch=-15}
   }
   n16 = "Transform (Modify) Bone - Bone: dyn_jacket_back_01_l"(
    ComponentPose=@n15
    Alpha=Get isSkydiveFloatingAlpha() -> (isSkydiveFloatingAlpha:double)
    Translation="0,0,0"
    Rotation="0,0,10"
    Scale="1,1,1"
   ) {
    rotationMode=BMM_Additive
    scaleMode=BMM_Additive
    scaleSpace=BCS_BoneSpace
    translationMode=BMM_Additive
    translationSpace=BCS_BoneSpace
    alphaBoolBlend {blendOption=Linear}
    boneToModify {boneName=dyn_jacket_back_01_l}
    rotation {roll=10}
   }
   n17 = "Transform (Modify) Bone - Bone: dyn_jacket_back_01_r"(ComponentPose=@n16, Alpha=1, Translation="0,0,0", Rotation="0,0,10", Scale="1,1,1") {
    rotationMode=BMM_Additive
    scaleMode=BMM_Additive
    scaleSpace=BCS_BoneSpace
    translationMode=BMM_Additive
    translationSpace=BCS_BoneSpace
    alphaBoolBlend {blendOption=Linear}
    boneToModify {boneName=dyn_jacket_back_01_r}
    rotation {roll=10}
   }
   n18 = Get emoting on low fps() -> (emoting on low fps:double)
   n19 = Look At - Bone: dyn_shirt_front_01_l_MSTR(ComponentPose=@n17, Alpha=@n18) {
    alphaBoolBlend {blendOption=Linear}
    alphaScaleBiasClamp {bInterpResult=true}
    boneToModify {boneName=dyn_shirt_front_01_l_MSTR}
    lookAtLocation {x=-25}
    lookAtTarget {boneReference {boneName=thigh_l}}
    lookAt_Axis {axis=<1 0>}
   }
   n20 = Look At - Bone: dyn_shirt_front_01_r_MSTR(ComponentPose=@n19, Alpha=@n18) {
    alphaBoolBlend {blendOption=Linear}
    alphaScaleBiasClamp {bInterpResult=true}
    boneToModify {boneName=dyn_shirt_front_01_r_MSTR}
    lookAtLocation {x=25}
    lookAtTarget {boneReference {boneName=thigh_r}}
    lookAt_Axis {axis=<1 0>}
   }
   Output Pose(Result=Component To Local(ComponentPose=@n20)) {layerGroup=None}
   n21 = Control Rig(Source=Copy Pose From Mesh() {bUseAttachedParent=true}) {
    alphaBoolBlend {blendOption=Linear}
    controlRigAssetReference {blueprintRigClass=&/M_MED_Rust_Remix_T1_neck_CtrlRig.M_MED_Rust_Remix_T1_neck_CtrlRig_C}
   }
   n22 = "Transform (Modify) Bone - Bone: dyn_shirt_front_01_r_MSTR"(
    ComponentPose=Local To Component(LocalPose=@n21)
    Alpha=Get thigh_rot_up_r() -> (thigh_rot_up_r:double)
    Translation="-10,0,10"
    Rotation="0,0,0"
    Scale="1,1,1"
   ) {
    rotationMode=BMM_Additive
    rotationSpace=BCS_BoneSpace
    scaleMode=BMM_Additive
    scaleSpace=BCS_BoneSpace
    translationMode=BMM_Additive
    translationSpace=BCS_BoneSpace
    alphaBoolBlend {blendOption=Linear}
    boneToModify {boneName=dyn_shirt_front_01_r_MSTR}
    translation {x=-10, z=10}
   }
   n23 = "Transform (Modify) Bone - Bone: dyn_shirt_front_01_MSTR"(ComponentPose=@n22, Alpha=Get thigh_rot_up_r(), Translation="-10,0,10", Rotation="0,0,0", Scale="1,1,1") {
    rotationMode=BMM_Additive
    rotationSpace=BCS_BoneSpace
    scaleMode=BMM_Additive
    scaleSpace=BCS_BoneSpace
    translationMode=BMM_Additive
    translationSpace=BCS_BoneSpace
    alphaBoolBlend {blendOption=Linear}
    boneToModify {boneName=dyn_shirt_front_01_MSTR}
    translation {x=-10, z=10}
   }
   n24 = "Transform (Modify) Bone - Bone: dyn_shirt_front_01_l_MSTR"(ComponentPose=@n23, Alpha=Get thigh_rot_up_l() -> (thigh_rot_up_l:double), Translation="-2,0,2", Rotation="0,0,0", Scale="1,1,1") {
    rotationMode=BMM_Additive
    rotationSpace=BCS_BoneSpace
    scaleMode=BMM_Additive
    scaleSpace=BCS_BoneSpace
    translationMode=BMM_Additive
    translationSpace=BCS_BoneSpace
    alphaBoolBlend {blendOption=Linear}
    boneToModify {boneName=dyn_shirt_front_01_l_MSTR}
    translation {x=-2, z=2}
   }
   n25 = "Transform (Modify) Bone - Bone: dyn_shirt_front_01_MSTR"!(ComponentPose=@n24, bAlphaBoolEnabled:bool=True, Translation="0,5,2.5", Rotation="0,0,0", Scale="1,1,1") -> (Pose) {
    alphaInputType=Bool
    rotationMode=BMM_Additive
    rotationSpace=BCS_BoneSpace
    scaleMode=BMM_Additive
    scaleSpace=BCS_BoneSpace
    translationMode=BMM_Additive
    translationSpace=BCS_BoneSpace
    alphaBoolBlend {blendInTime=0.2, blendOption=Linear, blendOutTime=0.2}
    alphaScaleBiasClamp {bInterpResult=true}
    boneToModify {boneName=dyn_shirt_front_01_MSTR}
    translation {y=5, z=2.5}
   }
   n26 = "Transform (Modify) Bone - Bone: dyn_shirt_front_01_l_MSTR"!(ComponentPose=@n25, bAlphaBoolEnabled:bool=True, Translation="-10,5,5", Rotation="0,0,0", Scale="1,1,1") -> (Pose) {
    alphaInputType=Bool
    rotationMode=BMM_Additive
    rotationSpace=BCS_BoneSpace
    scaleMode=BMM_Additive
    scaleSpace=BCS_BoneSpace
    translationMode=BMM_Additive
    translationSpace=BCS_BoneSpace
    alphaBoolBlend {blendInTime=0.2, blendOption=Linear, blendOutTime=0.2}
    alphaScaleBiasClamp {bInterpResult=true}
    boneToModify {boneName=dyn_shirt_front_01_l_MSTR}
    translation=<-10 5 5>
   }
   n27 = "Constraint - dyn_jacket_upper_01_l_MSTR (P)"(ComponentPose=@n26, Alpha=1, ConstraintWeights_0=1) {
    alphaBoolBlend {blendOption=Linear}
    boneToModify {boneName=dyn_jacket_upper_01_l_MSTR}
    constraintSetup [{targetBone {boneName=spine_04}}]
    constraintWeights [1]
   }
   n28 = "Constraint - dyn_jacket_upper_01_r_MSTR (P)"(ComponentPose=@n27, Alpha=1, ConstraintWeights_0=1) {
    alphaBoolBlend {blendOption=Linear}
    boneToModify {boneName=dyn_jacket_upper_01_r_MSTR}
    constraintSetup [{targetBone {boneName=spine_04}}]
    constraintWeights [1]
   }
   n29 = "Constraint - dyn_collar1_01_l_MSTR (P)"(ComponentPose=@n28, Alpha=1, ConstraintWeights_0=1) {
    alphaBoolBlend {blendOption=Linear}
    boneToModify {boneName=dyn_collar1_01_l_MSTR}
    constraintSetup [{targetBone {boneName=spine_04}}]
    constraintWeights [1]
   }
   n30 = "Constraint - dyn_collar1_01_r_MSTR (P)"(ComponentPose=@n29, Alpha=1, ConstraintWeights_0=1) {
    alphaBoolBlend {blendOption=Linear}
    boneToModify {boneName=dyn_collar1_01_r_MSTR}
    constraintSetup [{targetBone {boneName=spine_04}}]
    constraintWeights [1]
   }
   n31 = "Constraint - dyn_collar2_01_l_MSTR (PP)"!(ComponentPose=@n30, Alpha=1, ConstraintWeights_0=0.5, ConstraintWeights_1:float=0.5) -> (Pose) {
    alphaBoolBlend {blendOption=Linear}
    boneToModify {boneName=dyn_collar2_01_l_MSTR}
    constraintSetup [{targetBone {boneName=spine_04}}, {targetBone {boneName=neck_01}}]
    constraintWeights [1,1]
   }
   n32 = "Constraint - dyn_collar2_01_r_MSTR (PP)"!(ComponentPose=@n31, Alpha=1, ConstraintWeights_0=0.5, ConstraintWeights_1:float=0.5) -> (Pose) {
    alphaBoolBlend {blendOption=Linear}
    boneToModify {boneName=dyn_collar2_01_r_MSTR}
    constraintSetup [{targetBone {boneName=spine_04}}, {targetBone {boneName=neck_01}}]
    constraintWeights [1,1]
   }
   n33 = Anim Dynamics - Bone: dyn_wristFlap_01_r_MSTR(ComponentPose=@n32, Alpha=1, ExternalForce="0,0,0") {
    angularDampingOverride=1
    angularSpringConstant=2500
    bAngularSpring=true
    bOverrideAngularDamping=true
    bUseGravityOverride=true
    gravityScale=0.35
    simulationSpace=RootRelative
    alphaBoolBlend {blendOption=Linear}
    boundBone {boneName=dyn_wristFlap_01_r_MSTR}
    componentLinearAccScale=<0.5 0.5 0.5>
    componentLinearVelScale=<0.5 0.5 0.5>
    gravityOverride {z=-980}
    physicsBodyDefinitions [{
      boundBone {boneName=dyn_wristFlap_01_r_MSTR}
      boxExtents=<12.5 12.5 12.5>
      constraintSetup {angularLimitsMax=<0 0 25>, angularLimitsMin=<0 0 -25>, angularTarget=<1 0 0>}
      localJointOffset=<1 0 0>
     }]
    relativeSpaceBone {boneName=spine_05}
   }
   n34 = Anim Dynamics - Bone: dyn_wristFlap_01_l_MSTR(ComponentPose=@n33, Alpha=1, ExternalForce="0,0,0") {
    angularDampingOverride=1
    angularSpringConstant=2500
    bAngularSpring=true
    bOverrideAngularDamping=true
    bUseGravityOverride=true
    gravityScale=0.35
    simulationSpace=RootRelative
    alphaBoolBlend {blendOption=Linear}
    boundBone {boneName=dyn_wristFlap_01_l_MSTR}
    componentLinearAccScale=<0.5 0.5 0.5>
    componentLinearVelScale=<0.5 0.5 0.5>
    gravityOverride {z=-980}
    physicsBodyDefinitions [{
      boundBone {boneName=dyn_wristFlap_01_l_MSTR}
      boxExtents=<12.5 12.5 12.5>
      constraintSetup {angularLimitsMax=<0 0 25>, angularLimitsMin=<0 0 -25>, angularTarget=<1 0 0>}
      localJointOffset=<1 0 0>
     }]
    relativeSpaceBone {boneName=spine_05}
   }
   n35 = Anim Dynamics - Bone: dyn_legpouch_r_01(ComponentPose=@n34, Alpha=1, ExternalForce="0,0,0") {
    angularDampingOverride=0.98
    angularSpringConstant=2500
    bAngularSpring=true
    bOverrideAngularDamping=true
    gravityScale=0.35
    simulationSpace=RootRelative
    alphaBoolBlend {blendOption=Linear}
    boundBone {boneName=dyn_legpouch_r_01}
    componentLinearAccScale=<1 1 1>
    componentLinearVelScale=<1 1 1>
    gravityOverride {z=-980}
    physicsBodyDefinitions [{
      boundBone {boneName=dyn_legpouch_r_01}
      boxExtents=<15 15 15>
      constraintSetup {angularLimitsMax=<0 2.5 10>, angularLimitsMin=<0 -2.5 -10>, angularTarget=<0 0 0>}
      localJointOffset=<2 0 0>
     }]
   }
   n36 = Get bIsInFrontEnd()
   n37 = "Transform (Modify) Bone - Bone: dyn_shirt_front_01_l_MSTR"(ComponentPose=@n35, bAlphaBoolEnabled=@n36, Translation="0,0,0.5", Rotation="0,0,0", Scale="1,1,1") {
    alphaInputType=Bool
    rotationMode=BMM_Additive
    rotationSpace=BCS_BoneSpace
    scaleMode=BMM_Additive
    scaleSpace=BCS_BoneSpace
    translationMode=BMM_Additive
    translationSpace=BCS_BoneSpace
    alphaBoolBlend {blendInTime=0.1, blendOption=Linear, blendOutTime=0.1}
    boneToModify {boneName=dyn_shirt_front_01_l_MSTR}
    translation {z=0.5}
   }
   n38 = "Transform (Modify) Bone - Bone: dyn_shirt_front_01_MSTR"(ComponentPose=@n37, bAlphaBoolEnabled=@n36, Translation="0,0,1.5", Rotation="0,0,0", Scale="1,1,1") {
    alphaInputType=Bool
    rotationMode=BMM_Additive
    rotationSpace=BCS_BoneSpace
    scaleMode=BMM_Additive
    scaleSpace=BCS_BoneSpace
    translationMode=BMM_Additive
    translationSpace=BCS_BoneSpace
    alphaBoolBlend {blendInTime=0.1, blendOption=Linear, blendOutTime=0.1}
    boneToModify {boneName=dyn_shirt_front_01_MSTR}
    translation {z=1.5}
   }
   n39 = "Transform (Modify) Bone - Bone: dyn_shirt_front_01_r_MSTR"!(ComponentPose=@n38, bAlphaBoolEnabled:bool=@n36, Translation="0,0,4", Rotation="0,0,0", Scale="1,1,1") -> (Pose) {
    alphaInputType=Bool
    rotationMode=BMM_Additive
    rotationSpace=BCS_BoneSpace
    scaleMode=BMM_Additive
    scaleSpace=BCS_BoneSpace
    translationMode=BMM_Additive
    translationSpace=BCS_BoneSpace
    alphaBoolBlend {blendInTime=0.1, blendOption=Linear, blendOutTime=0.1}
    boneToModify {boneName=dyn_shirt_front_01_r_MSTR}
    translation {z=4}
   }
   n40 = RigidBody!(ComponentPose=@n39, Alpha, OverrideWorldGravity:Vector="0,0,-980", ExternalForce) -> (Pose) {
    bClampLinearTranslationLimitToRefPose=true
    bOverrideWorldGravity=true
    overridePhysicsAsset=&/M_MED_Rust_Remix_FrontShirt_Physics
    simulationSpace=BaseBoneSpace
    alphaBoolBlend {blendOption=Linear}
    baseBoneRef {boneName=pelvis}
    componentAppliedLinearAccClamp=<1200 1200 1200>
    componentLinearAccScale=<0.3 0.3 0.3>
    componentLinearVelScale=<0.2 0.2 0.2>
    overrideWorldGravity {z=-980}
    simSpaceSettings {
     maxAngularAcceleration=1000
     maxAngularVelocity=3
     maxLinearAcceleration=700
     maxLinearVelocity=500
     velocityScaleZ=0
     worldAlpha=0.1
     externalLinearDragV=<0.2 0.2>
    }
   }
   n41 = RigidBody(ComponentPose=@n40, OverrideWorldGravity="0,0,-980") {
    bClampLinearTranslationLimitToRefPose=true
    bOverrideWorldGravity=true
    overridePhysicsAsset=&/M_MED_Rust_Remix_shirtBack_Physics
    simulationSpace=BaseBoneSpace
    alphaBoolBlend {blendOption=Linear}
    baseBoneRef {boneName=pelvis}
    componentAppliedLinearAccClamp=<8000 1200 1200>
    componentLinearAccScale=<0.5 0.2 0.25>
    componentLinearVelScale=<0.5 0.1 0.2>
    overrideWorldGravity {z=-980}
    simSpaceSettings {
     maxAngularAcceleration=1000
     maxAngularVelocity=3
     maxLinearAcceleration=700
     maxLinearVelocity=500
     velocityScaleZ=0
     worldAlpha=0.5
     externalLinearDragV=<0.2 0.2>
    }
   }
   n42 = RigidBody(ComponentPose=@n41, OverrideWorldGravity="0,0,-980") {
    bClampLinearTranslationLimitToRefPose=true
    bOverrideWorldGravity=true
    overridePhysicsAsset=&/M_MED_Rust_Remix_jacketBack_Physics
    simulationSpace=BaseBoneSpace
    alphaBoolBlend {blendOption=Linear}
    baseBoneRef {boneName=pelvis}
    componentAppliedLinearAccClamp=<5000 1200 1200>
    componentLinearAccScale=<0.4 0.2 0.25>
    componentLinearVelScale=<0.4 0.1 0.2>
    overrideWorldGravity {z=-980}
    simSpaceSettings {
     maxAngularAcceleration=1000
     maxAngularVelocity=3
     maxLinearAcceleration=700
     maxLinearVelocity=500
     velocityScaleZ=0
     worldAlpha=0.15
     externalLinearDragV=<0.2 0.2>
    }
   }
   n43 = "Transform (Modify) Bone - Bone: dyn_jacket_back_01_MSTR"(ComponentPose=@n42, Alpha=1, Translation="0,0,2.5", Rotation="0,0,0", Scale="1,1,1") {
    rotationMode=BMM_Additive
    rotationSpace=BCS_BoneSpace
    scaleMode=BMM_Additive
    scaleSpace=BCS_BoneSpace
    translationMode=BMM_Additive
    translationSpace=BCS_BoneSpace
    alphaBoolBlend {blendOption=Linear}
    boneToModify {boneName=dyn_jacket_back_01_MSTR}
    translation {z=2.5}
   }
   n44 = "Constraint - dyn_jacket_front_01_r (R)"(ComponentPose=@n43, Alpha=0.2, ConstraintWeights_0=1) {
    alphaBoolBlend {blendOption=Linear}
    boneToModify {boneName=dyn_jacket_front_01_r}
    constraintSetup [{transformType=Rotation, targetBone {boneName=thigh_r}}]
    constraintWeights [1]
   }
   n45 = "Constraint - dyn_jacket_front_02_r (R)"(ComponentPose=@n44, Alpha=1, ConstraintWeights_0=1) {
    alphaBoolBlend {blendOption=Linear}
    boneToModify {boneName=dyn_jacket_front_02_r}
    constraintSetup [{transformType=Rotation, targetBone {boneName=dyn_shirt_front_01_r}}]
    constraintWeights [1]
   }
   n46 = "Constraint - dyn_jacket_front_04_r (PPPPPPPPPPPPPPPPPPPPPPPPPPPPPPPPP)"!(
    ComponentPose=@n45
    Alpha=1
    ConstraintWeights_0=0.0097
    ConstraintWeights_1:float=0.0017
    ConstraintWeights_2:float=0
    ConstraintWeights_3:float=0.0005
    ConstraintWeights_4:float=0.0116
    ConstraintWeights_5:float=0
    ConstraintWeights_6:float=0
    ConstraintWeights_7:float=0
    ConstraintWeights_8:float=0.0155
    ConstraintWeights_9:float=0.0746
    ConstraintWeights_10:float=0.0029
    ConstraintWeights_11:float=0
    ConstraintWeights_12:float=0
    ConstraintWeights_13:float=0
    ConstraintWeights_14:float=0
    ConstraintWeights_15:float=0
    ConstraintWeights_16:float=0
    ConstraintWeights_17:float=0
    ConstraintWeights_18:float=0.0014
    ConstraintWeights_19:float=0.0026
    ConstraintWeights_20:float=0.0027
    ConstraintWeights_21:float=0.0039
    ConstraintWeights_22:float=0.158
    ConstraintWeights_23:float=0.1987
    ConstraintWeights_24:float=0.1492
    ConstraintWeights_25:float=0.0153
    ConstraintWeights_26:float=0
    ConstraintWeights_27:float=0
    ConstraintWeights_28:float=0
    ConstraintWeights_29:float=0.085
    ConstraintWeights_30:float=0.1794
    ConstraintWeights_31:float=0.076
    ConstraintWeights_32:float=0.0113
   ) -> (Pose) {
    alphaBoolBlend {blendOption=Linear}
    boneToModify {boneName=dyn_jacket_front_04_r}
    constraintSetup [
     {targetBone {boneName=spine_03}}
     {targetBone {boneName=spine_04}}
     {targetBone {boneName=upperarm_r}}
     {targetBone {boneName=clavicle_r}}
     {targetBone {boneName=spine_02}}
     {targetBone {boneName=spine_05}}
     {targetBone {boneName=clavicle_l}}
     {targetBone {boneName=upperarm_twist_01_l}}
     {targetBone {boneName=spine_01}}
     {targetBone {boneName=pelvis}}
     {targetBone {boneName=thigh_twist_01_r}}
     {targetBone {boneName=thigh_twist_01_l}}
     {targetBone {boneName=thigh_r}}
     {targetBone {boneName=thigh_l}}
     {targetBone {boneName=upperarm_l}}
     {targetBone {boneName=upperarm_twist_02_l}}
     {targetBone {boneName=upperarm_twist_01_r}}
     {targetBone {boneName=upperarm_twist_02_r}}
     {targetBone {boneName=dyn_shirt_front_04_l}}
     {targetBone {boneName=dyn_shirt_front_03_l}}
     {targetBone {boneName=dyn_shirt_front_02_l}}
     {targetBone {boneName=dyn_shirt_front_01_l}}
     {targetBone {boneName=dyn_shirt_front_01_r}}
     {targetBone {boneName=dyn_shirt_front_02_r}}
     {targetBone {boneName=dyn_shirt_front_03_r}}
     {targetBone {boneName=dyn_shirt_front_04_r}}
     {targetBone {boneName=dyn_jacket_back_01_r}}
     {targetBone {boneName=dyn_jacket_back_01_l}}
     {targetBone {boneName=dyn_collar_front_01}}
     {targetBone {boneName=dyn_shirt_front_01}}
     {targetBone {boneName=dyn_shirt_front_02}}
     {targetBone {boneName=dyn_shirt_front_03}}
     {targetBone {boneName=dyn_shirt_front_04}}
    ]
    constraintWeights [1,1,1,1,1,1,1,1,1,1,1,1,1,1,1,1,1,1,1,1,1,1,1,1,1,1,1,1,1,1,1,1,1]
   }
   n47 = "Constraint - dyn_jacket_front_02_l (PP)"!(ComponentPose=@n46, Alpha=1, ConstraintWeights_0=0.5, ConstraintWeights_1:float=0.5) -> (Pose) {
    alphaBoolBlend {blendOption=Linear}
    boneToModify {boneName=dyn_jacket_front_02_l}
    constraintSetup [{targetBone {boneName=dyn_shirt_front_01_l}}, {targetBone {boneName=dyn_shirt_front_01}}]
    constraintWeights [1,1]
   }
   n48 = "Constraint - dyn_jacket_front_03_l (P)"(ComponentPose=@n47, Alpha=1, ConstraintWeights_0=1) {
    alphaBoolBlend {blendOption=Linear}
    boneToModify {boneName=dyn_jacket_front_03_l}
    constraintSetup [{targetBone {boneName=dyn_shirt_front_01_l}}]
    constraintWeights [1]
   }
   n49 = "Constraint - dyn_jacket_front_03_l (P)"(ComponentPose=@n48, Alpha=1, ConstraintWeights_0=1) {
    alphaBoolBlend {blendOption=Linear}
    boneToModify {boneName=dyn_jacket_front_03_l}
    constraintSetup [{targetBone {boneName=dyn_shirt_front_01_l}}]
    constraintWeights [1]
   }
   n50 = "Constraint - dyn_jacket_front_04_l (P)"(ComponentPose=@n49, Alpha=1, ConstraintWeights_0=1) {
    alphaBoolBlend {blendOption=Linear}
    boneToModify {boneName=dyn_jacket_front_04_l}
    constraintSetup [{targetBone {boneName=dyn_shirt_front_02_l}}]
    constraintWeights [1]
   }
   n51 = "Transform (Modify) Bone - Bone: dyn_jacket1_01_l_MSTR"!(ComponentPose=@n50, bAlphaBoolEnabled:bool=True, Translation="-6.300,1.263,-1.019", Rotation="42.094,25.209,-11.466", Scale="1,1,1") -> (Pose) {
    alphaInputType=Bool
    rotationMode=BMM_Additive
    rotationSpace=BCS_BoneSpace
    scaleMode=BMM_Additive
    scaleSpace=BCS_BoneSpace
    translationMode=BMM_Additive
    translationSpace=BCS_BoneSpace
    alphaBoolBlend {blendInTime=0.1, blendOption=Linear, blendOutTime=0.1}
    boneToModify {boneName=dyn_jacket1_01_l_MSTR}
    rotation=<rot 42.094 25.209 -11.466>
    translation=<-6.3 1.263 -1.019>
   }
   n52 = "Constraint - dyn_jacket2_01_l_MSTR (R)"(ComponentPose=@n51, Alpha=0.1, ConstraintWeights_0=1) {
    alphaBoolBlend {blendOption=Linear}
    boneToModify {boneName=dyn_jacket2_01_l_MSTR}
    constraintSetup [{transformType=Rotation, targetBone {boneName=dyn_jacket_back_01_l}}]
    constraintWeights [1]
   }
   n53 = "Constraint - dyn_jacket_back_01_MSTR (RR)"!(ComponentPose=@n52, Alpha=0.25, ConstraintWeights_0=1, ConstraintWeights_1:float=1) -> (Pose) {
    alphaBoolBlend {blendOption=Linear}
    boneToModify {boneName=dyn_jacket_back_01_MSTR}
    constraintSetup [{transformType=Rotation, targetBone {boneName=dyn_jacket_back_01_l}}, {transformType=Rotation, targetBone {boneName=dyn_jacket_back_01_r}}]
    constraintWeights [1,1]
   }
   n54 = "Constraint - dyn_jacket2_01_r_MSTR (R)"(ComponentPose=@n53, Alpha=0.1, ConstraintWeights_0=1) {
    alphaBoolBlend {blendOption=Linear}
    boneToModify {boneName=dyn_jacket2_01_r_MSTR}
    constraintSetup [{transformType=Rotation, targetBone {boneName=dyn_jacket_back_01_r}}]
    constraintWeights [1]
   }
   n55 = "Transform (Modify) Bone - Bone: dyn_jacket_back_01_l_MSTR"!(ComponentPose=@n54, bAlphaBoolEnabled:bool=True, Translation="0,0,1", Rotation="0,0,0", Scale="1,1,1") -> (Pose) {
    alphaInputType=Bool
    rotationMode=BMM_Additive
    rotationSpace=BCS_BoneSpace
    scaleMode=BMM_Additive
    scaleSpace=BCS_BoneSpace
    translationMode=BMM_Additive
    translationSpace=BCS_BoneSpace
    alphaBoolBlend {blendInTime=0.1, blendOption=Linear, blendOutTime=0.1}
    boneToModify {boneName=dyn_jacket_back_01_l_MSTR}
    translation {z=1}
   }
   n56 = "Transform (Modify) Bone - Bone: dyn_jacket_back_01_r_MSTR"!(ComponentPose=@n55, bAlphaBoolEnabled:bool=True, Translation="0,0,-1", Rotation="0,0,0", Scale="1,1,1") -> (Pose) {
    alphaInputType=Bool
    rotationMode=BMM_Additive
    rotationSpace=BCS_BoneSpace
    scaleMode=BMM_Additive
    scaleSpace=BCS_BoneSpace
    translationMode=BMM_Additive
    translationSpace=BCS_BoneSpace
    alphaBoolBlend {blendInTime=0.1, blendOption=Linear, blendOutTime=0.1}
    boneToModify {boneName=dyn_jacket_back_01_r_MSTR}
    translation {z=-1}
   }
   n57 = "Transform (Modify) Bone - Bone: dyn_shirt_front_01_r_MSTR"(ComponentPose=@n56, bAlphaBoolEnabled=True, Translation="0,0,0", Rotation="0,0,0", Scale="0.8,1,1") {
    alphaInputType=Bool
    rotationMode=BMM_Additive
    rotationSpace=BCS_BoneSpace
    scaleMode=BMM_Additive
    scaleSpace=BCS_BoneSpace
    translationMode=BMM_Additive
    translationSpace=BCS_BoneSpace
    alphaBoolBlend {blendInTime=0.1, blendOption=Linear, blendOutTime=0.1}
    boneToModify {boneName=dyn_shirt_front_01_r_MSTR}
    scale {x=0.8}
   }
   n58 = "Transform (Modify) Bone - Bone: dyn_shirt_front_01_MSTR"(ComponentPose=@n57, bAlphaBoolEnabled=True, Translation="0,0,0", Rotation="0,0,0", Scale="0.8,1,1") {
    alphaInputType=Bool
    rotationMode=BMM_Additive
    rotationSpace=BCS_BoneSpace
    scaleMode=BMM_Additive
    scaleSpace=BCS_BoneSpace
    translationMode=BMM_Additive
    translationSpace=BCS_BoneSpace
    alphaBoolBlend {blendInTime=0.1, blendOption=Linear, blendOutTime=0.1}
    boneToModify {boneName=dyn_shirt_front_01_MSTR}
    scale {x=0.8}
   }
   n59 = "Transform (Modify) Bone - Bone: dyn_shirt_front_01_l_MSTR"(ComponentPose=@n58, bAlphaBoolEnabled=True, Translation="0,0,0", Rotation="0,0,0", Scale="0.8,1,1") {
    alphaInputType=Bool
    rotationMode=BMM_Additive
    rotationSpace=BCS_BoneSpace
    scaleMode=BMM_Additive
    scaleSpace=BCS_BoneSpace
    translationMode=BMM_Additive
    translationSpace=BCS_BoneSpace
    alphaBoolBlend {blendInTime=0.1, blendOption=Linear, blendOutTime=0.1}
    boneToModify {boneName=dyn_shirt_front_01_l_MSTR}
    scale {x=0.8}
   }
   n60 = "Transform (Modify) Bone - Bone: dyn_collar2_01_l_MSTR"(ComponentPose=@n59, Alpha=Get HeadTurn_L() -> (HeadTurn_L:double), Translation="0,0,-1", Rotation="-10,0,0", Scale="1,1,1") {
    rotationMode=BMM_Additive
    rotationSpace=BCS_BoneSpace
    scaleMode=BMM_Additive
    scaleSpace=BCS_BoneSpace
    translationMode=BMM_Additive
    translationSpace=BCS_BoneSpace
    alphaBoolBlend {blendInTime=0.1, blendOption=Linear, blendOutTime=0.1}
    boneToModify {boneName=dyn_collar2_01_l_MSTR}
    rotation {pitch=-10}
    translation {z=-1}
   }
   n61 = "Transform (Modify) Bone - Bone: dyn_collar1_01_l_MSTR"(ComponentPose=@n60, Alpha=Get HeadTurn_L(), Translation="0,0,0", Rotation="-5,0,0", Scale="1,1,1") {
    rotationMode=BMM_Additive
    rotationSpace=BCS_BoneSpace
    scaleMode=BMM_Additive
    scaleSpace=BCS_BoneSpace
    translationMode=BMM_Additive
    translationSpace=BCS_BoneSpace
    alphaBoolBlend {blendInTime=0.1, blendOption=Linear, blendOutTime=0.1}
    boneToModify {boneName=dyn_collar1_01_l_MSTR}
    rotation {pitch=-5}
   }
   Save cached pose 'preSimCache'(Pose=Component To Local(ComponentPose=@n61)) {cacheName=preSimCache}
   n62 = "Transform (Modify) Bone - Bone: dyn_jacket2_01_l"(ComponentPose=Local To Component(LocalPose=Use cached pose 'preSimCache'()), Alpha=1, Translation="0,0,0", Rotation="5,0,0", Scale="1,1,1") {
    rotationMode=BMM_Additive
    rotationSpace=BCS_BoneSpace
    scaleMode=BMM_Additive
    scaleSpace=BCS_BoneSpace
    translationMode=BMM_Additive
    translationSpace=BCS_BoneSpace
    alphaBoolBlend {blendOption=Linear}
    boneToModify {boneName=dyn_jacket2_01_l}
    rotation {pitch=5}
   }
   n63 = "Transform (Modify) Bone - Bone: dyn_jacket_back_01"(ComponentPose=@n62, Alpha=1, Translation="0,0,0", Rotation="5,0,0", Scale="1,1,1") {
    rotationMode=BMM_Additive
    rotationSpace=BCS_BoneSpace
    scaleMode=BMM_Additive
    scaleSpace=BCS_BoneSpace
    translationMode=BMM_Additive
    translationSpace=BCS_BoneSpace
    alphaBoolBlend {blendOption=Linear}
    boneToModify {boneName=dyn_jacket_back_01}
    rotation {pitch=5}
   }
   n64 = "Transform (Modify) Bone - Bone: dyn_jacket2_01_r"(ComponentPose=@n63, Alpha=1, Translation="0,0,0", Rotation="5,0,0", Scale="1,1,1") {
    rotationMode=BMM_Additive
    rotationSpace=BCS_BoneSpace
    scaleMode=BMM_Additive
    scaleSpace=BCS_BoneSpace
    translationMode=BMM_Additive
    translationSpace=BCS_BoneSpace
    alphaBoolBlend {blendOption=Linear}
    boneToModify {boneName=dyn_jacket2_01_r}
    rotation {pitch=5}
   }
   n65 = "Transform (Modify) Bone - Bone: dyn_shirt_front_01_r"(ComponentPose=@n64, Alpha=1, Translation="0,0,2", Rotation="10,0,0", Scale="1,1,1") {
    rotationMode=BMM_Additive
    rotationSpace=BCS_BoneSpace
    scaleMode=BMM_Additive
    scaleSpace=BCS_BoneSpace
    translationMode=BMM_Additive
    translationSpace=BCS_BoneSpace
    alphaBoolBlend {blendOption=Linear}
    boneToModify {boneName=dyn_shirt_front_01_r}
    rotation {pitch=10}
    translation {z=2}
   }
   n66 = "Transform (Modify) Bone - Bone: dyn_shirt_front_01_l"(ComponentPose=@n65, Alpha=1, Translation="0,0,2", Rotation="10,0,0", Scale="1,1,1") {
    rotationMode=BMM_Additive
    rotationSpace=BCS_BoneSpace
    scaleMode=BMM_Additive
    scaleSpace=BCS_BoneSpace
    translationMode=BMM_Additive
    translationSpace=BCS_BoneSpace
    alphaBoolBlend {blendOption=Linear}
    boneToModify {boneName=dyn_shirt_front_01_l}
    rotation {pitch=10}
    translation {z=2}
   }
   n67 = "Constraint - dyn_jacket_front_01_r (RR)"!(ComponentPose=@n66, Alpha=1, ConstraintWeights_0=0.5, ConstraintWeights_1:float=0.5) -> (Pose) {
    alphaBoolBlend {blendOption=Linear}
    boneToModify {boneName=dyn_jacket_front_01_r}
    constraintSetup [{transformType=Rotation, targetBone {boneName=dyn_shirt_front_02_r}}, {transformType=Rotation, targetBone {boneName=pelvis}}]
    constraintWeights [1,1]
   }
   n68 = "Constraint - dyn_jacket_front_01_l (RR)"!(ComponentPose=@n67, Alpha=1, ConstraintWeights_0=0.5, ConstraintWeights_1:float=0.5) -> (Pose) {
    alphaBoolBlend {blendOption=Linear}
    boneToModify {boneName=dyn_jacket_front_01_l}
    constraintSetup [{transformType=Rotation, targetBone {boneName=dyn_shirt_front_02_l}}, {transformType=Rotation, targetBone {boneName=pelvis}}]
    constraintWeights [1,1]
   }
   n69 = "Constraint - dyn_jacket2_01_l (PR)"!(ComponentPose=@n68, Alpha=1, ConstraintWeights_0=1, ConstraintWeights_1:float=0) -> (Pose) {
    alphaBoolBlend {blendOption=Linear}
    boneToModify {boneName=dyn_jacket2_01_l}
    constraintSetup [{targetBone {boneName=dyn_jacket_back_01_l}}, {transformType=Rotation, targetBone {boneName=pelvis}}]
    constraintWeights [1,1]
   }
   n70 = "Constraint - dyn_jacket_back_01 (PP)"!(ComponentPose=@n69, Alpha=1, ConstraintWeights_0=0.5, ConstraintWeights_1:float=0.5) -> (Pose) {
    alphaBoolBlend {blendOption=Linear}
    boneToModify {boneName=dyn_jacket_back_01}
    constraintSetup [{targetBone {boneName=dyn_jacket_back_01_l}}, {targetBone {boneName=dyn_jacket_back_01_r}}]
    constraintWeights [1,1]
   }
   n71 = "Constraint - dyn_jacket2_01_r (PR)"!(ComponentPose=@n70, Alpha=1, ConstraintWeights_0=1, ConstraintWeights_1:float=0) -> (Pose) {
    alphaBoolBlend {blendOption=Linear}
    boneToModify {boneName=dyn_jacket2_01_r}
    constraintSetup [{targetBone {boneName=dyn_jacket_back_01_r}}, {transformType=Rotation, targetBone {boneName=pelvis}}]
    constraintWeights [1,1]
   }
   Save cached pose 'lod2Cache'(Pose=Component To Local(ComponentPose=@n71)) {cacheName=lod2Cache}
  ]}
 EventGraph {nodes [
   n1 = Event BlueprintUpdateAnimation()
   > n2 = Sequence!(execute) -> (then_0, then_1, then_2:Exec, then_3:Exec, then_4:Exec, then_5:Exec, then_6:Exec, then_7:Exec, then_8:Exec, then_9:Exec, then_10:Exec)
   n3 = Get bRigidBodyEnabled()
   n4 = Get CurrentLOD()
   > n5 = Set isMobileLOD(isMobileLOD:bool=OR Boolean(A=NOT Boolean(A=@n3), B="integer > integer"(A=@n4, B=1) {is=title K2Node_CallFunction})) -> (Output_Get:bool)
   > Set isMobileLODAlpha(isMobileLODAlpha:double="To Float (Boolean)"(InBool=@n5)) -> (Output_Get:double)
   n6 = OR Boolean(A=Get bCanPlayCustomAnimations() -> (bCanPlayCustomAnimations:bool), B=Get bIsPlayingEmote())
   > Set isEmotingAlpha(isEmotingAlpha:double="To Float (Boolean)"(InBool=@n6)) -> (Output_Get:double)
   > Set isEmotingAlphaInv(isEmotingAlphaInv:double="To Float (Boolean)"(InBool=NOT Boolean(A=@n6))) -> (Output_Get:double)
   n7 = GetOwnerSkeletalMeshComponent()
   n8 = BreakRotator(InRot=BreakTransform(InTransform=GetDeltaTransformFromRefPose(self=@n7, BoneName=upperarm_l, BaseName=clavicle_l)).Rotation)
   Set armSwingDelta_r(execute=@n2.then_1, armSwingDelta_r:double=MapRangeClamped(Value="Absolute (Float)"(A=@n8.Yaw), InRangeA=5, InRangeB=70)) -> (Output_Get:double)
   n9 = Get bIsTargeting()
   > Set isAimingClampedAlpha(isAimingClampedAlpha:double="To Float (Boolean)"(InBool=@n9)) -> (Output_Get:double)
   > Set isNotAimingAlpha(isNotAimingAlpha:double=SelectFloat(A=1, B=0.5, bPickA=NOT Boolean(A=@n9))) -> (Output_Get:double)
   n10 = Set low fps(execute=@n2.then_10, low fps:double=MapRangeClamped(Value=float / float(A=1, B=@n1.DeltaTimeX), InRangeA=15, InRangeB=50, OutRangeA=1)) -> (Output_Get:double)
   > Set emoting on low fps(emoting on low fps:double=float * float(A=@n10, B:bool=Get bIsPlayingEmote()) {is=title K2Node_PromotableOperator}) -> (Output_Get:double)
   n11 = MapRangeClamped(
    Value=GetTwistAndSwingAngleOfDeltaRotationFromRefPose(self=GetOwnerSkeletalMeshComponent(), BoneName=thigh_l).OutSwingAngle
    InRangeA=50
    InRangeB=100
   )
   Set spineSwingAngleAlpha(execute=@n2.then_2, spineSwingAngleAlpha:double=@n11) -> (Output_Get:double)
   n12 = GetOwnerSkeletalMeshComponent()
   > Set headSwingAngleAlpha(headSwingAngleAlpha:double=MapRangeClamped(Value=GetTwistAndSwingAngleOfDeltaRotationFromRefPose(self=@n12, BoneName=head).OutSwingAngle, InRangeA=20, InRangeB=45)) -> (Output_Get:double)
   n13 = GetTwistAndSwingAngleOfDeltaRotationFromRefPose(self=GetOwnerSkeletalMeshComponent(), BoneName=thigh_r)
   n14 = NOT Boolean(A=Get bIsCrouching())
   n15 = NOT Boolean(A=Get bIsSkyDiving())
   > Set R_legSwingAlpha(R_legSwingAlpha:double=SelectFloat(A=MapRangeClamped(Value=@n13.OutSwingAngle, InRangeA=70, InRangeB=154, OutRangeA=1, OutRangeB=0), B=0.15, bPickA=AND Boolean(A=@n14, B=@n15))) -> (Output_Get:double)
   n16 = Get bIsSkydiveFloating()
   Set isSkydiveFloatingAlpha(execute=@n2.then_3, isSkydiveFloatingAlpha:double="To Float (Boolean)"(InBool=@n16)) -> (Output_Get:double)
   n17 = NOT Boolean(A=Get bIsSkydiveDiveMode())
   n18 = NOT Boolean(A=@n16)
   n19 = Get bIsSkyDiving()
   > Set isSkydiveNotFloating(isSkydiveNotFloating:double="To Float (Boolean)"(InBool=AND Boolean!(A=@n18, B=@n19, C:bool=@n17) -> (ReturnValue))) -> (Output_Get:double)
   n20 = Get bIsSkyDiving()
   > Set isNotSkydiving(isNotSkydiving:double="To Float (Boolean)"(InBool=NOT Boolean(A=@n20))) -> (Output_Get:double)
   n21 = Get SkydivingExternalForceInput()
   Set windVector(
    execute=@n2.then_4
    windVector:Vector=vector * float(
     A=Break Fort Anim Input Skydiving External Force(FortAnimInput_SkydivingExternalForce=@n21) -> (
      bUseSkydivingVectorForce=~
      SkydivingVectorForce=~
      HeadToPelvisDirection=~
      FloatingMultiplier=~
      FloatingAdditive=~
      DivingMultiplier=~
      DivingAdditive=~
      ParachutingMultiplier=~
      ParachutingAdditive=~
     ).SkydivingVectorForce
     B=3
    )
   ) -> (Output_Get:Vector)
   n22 = Gravity Override State Function DEPRECATED(
    execute=@n2.then_5
    CharacterPartAnimInstance=Self-Reference()
    GameTimeInSeconds=GetGameTimeInSeconds()
    GravityOverrideParams=Get Gravity Override Params()
   )
   > Set GravityOverride(GravityOverride=@n22)
   Set Thigh R Rot(
    execute=@n2.then_8
    Thigh R Rot:double=GetAlphaFromReferenceTransformDelta(
     SkeletalMeshComponent=Get Body Mesh() -> (Body Mesh:Skeletal Mesh Component Object Reference)
     BoneName=thigh_r
     BaseName=pelvis
     InputRange=(X=0,Y=100)
     TwistAxis="1,0,0"
    )
   ) -> (Output_Get:double)
   n23 = GetOwnerSkeletalMeshComponent()
   n24 = BreakRotator(InRot=BreakTransform(InTransform=GetDeltaTransformFromRefPose(self=@n23, BoneName=thigh_r, BaseName=pelvis)).Rotation)
   Set thigh_rot_up_r(execute=@n2.then_6, thigh_rot_up_r:double=MapRangeClamped(Value=@n24.Yaw, InRangeA=-15, InRangeB=-100)) -> (Output_Get:double)
   > Set thigh_rot_inr_r(thigh_rot_inr_r:double=MapRangeClamped(Value=@n24.Pitch, InRangeA=0, InRangeB=-10, OutRangeA=1, OutRangeB=0)) -> (Output_Get:double)
   > Set thigh_rot_otr_r(thigh_rot_otr_r:double=MapRangeClamped(Value=@n24.Pitch, InRangeA=20, InRangeB=50)) -> (Output_Get:double)
   n25 = GetOwnerSkeletalMeshComponent()
   n26 = BreakRotator(InRot=BreakTransform(InTransform=GetDeltaTransformFromRefPose(self=@n25, BoneName=thigh_l, BaseName=pelvis)).Rotation)
   > Set thigh_rot_up_l(thigh_rot_up_l:double=MapRangeClamped(Value=@n26.Yaw, InRangeA=-10, InRangeB=-100)) -> (Output_Get:double)
   > Set thigh_rot_inr_l(thigh_rot_inr_l:double=MapRangeClamped(Value=@n26.Pitch, InRangeA=0, InRangeB=-10, OutRangeA=1, OutRangeB=0)) -> (Output_Get:double)
   > Set thigh_rot_otr_l(thigh_rot_otr_l:double=MapRangeClamped(Value=@n26.Pitch, InRangeA=20, InRangeB=50)) -> (Output_Get:double)
   n27 = GetOwnerSkeletalMeshComponent()
   n28 = BreakRotator(InRot=BreakTransform(InTransform=GetDeltaTransformFromRefPose(self=@n27, BoneName=head, BaseName=neck_02)).Rotation)
   > Set HeadTurn_L(HeadTurn_L:double=MapRangeClamped(Value=@n28, InRangeA=0, InRangeB=15)) -> (Output_Get:double)
   > Set HeadTurn_R(HeadTurn_R:double=MapRangeClamped(Value=@n28, InRangeA=0, InRangeB=-15)) -> (Output_Get:double)
   Set thigh L Rot(
    execute=@n2.then_7
    thigh L Rot:double=GetAlphaFromReferenceTransformDelta(SkeletalMeshComponent=Get Body Mesh(), BoneName=thigh_l, BaseName=pelvis, InputRange=(X=0,Y=-100), TwistAxis="1,0,0")
   ) -> (Output_Get:double)
   Set Body Mesh(execute=@n2.then_9, Body Mesh:Skeletal Mesh Component Object Reference=GetPartSkeletalMeshComponent()) -> (Output_Get:Skeletal Mesh Component Object Reference)
   Event BlueprintInitializeAnimation()
   > Set isCrouchingAlpha(isCrouchingAlpha:double=0) -> (Output_Get:double)
   > Set isNotCrouchingAlpha(isNotCrouchingAlpha:double=1) -> (Output_Get:double)
   > Set isNotCrouchingJacketAlpha_08(isNotCrouchingJacketAlpha_08:double=0.8) -> (Output_Get:double)
   > ResetDynamics(InTeleportType=TeleportPhysics)
   > ResetDynamics(InTeleportType=ResetPhysics)
   Event OnEndTargeting() {is=title K2Node_Event}
   > Set isTargetingAlpha(isTargetingAlpha:double=0) -> (Output_Get:double)
   Event OnBeginCrouching()
   > Set isCrouchingAlpha(isCrouchingAlpha=1)
   > Set isNotCrouchingAlpha(isNotCrouchingAlpha=0)
   > Set isNotCrouchingJacketAlpha_08(isNotCrouchingJacketAlpha_08=0)
   > ResetDynamics(InTeleportType=TeleportPhysics)
   > ResetDynamics(InTeleportType=ResetPhysics)
   Event OnBeginTargeting() {is=title K2Node_Event}
   > Set isTargetingAlpha(isTargetingAlpha=1)
   Event OnEndCrouching()
   > Set isCrouchingAlpha(isCrouchingAlpha=0)
   > Set isNotCrouchingAlpha(isNotCrouchingAlpha=1)
   > Set isNotCrouchingJacketAlpha_08(isNotCrouchingJacketAlpha_08=0.8)
   > Delay(Duration=0.125)
   > ResetDynamics(InTeleportType=TeleportPhysics)
   > ResetDynamics(InTeleportType=ResetPhysics)
   SelectFloat(A=1, B=@n11, bPickA=Get bIsSkyDiving())
  ]}
]
variables [
 armSwingDelta_r {type=number}
 Body Mesh {type=SkeletalMeshComponent}
 emoting on low fps {type=number}
 Gravity Override Params {type=GravityOverrideParamsStruct}
 GravityOverride {type=Vector}
 headSwingAngleAlpha {type=number}
 HeadTurn_L {type=number}
 HeadTurn_R {type=number}
 isAimingClampedAlpha {type=number}
 isCrouchingAlpha {type=number}
 isEmotingAlpha {type=number}
 isEmotingAlphaInv {type=number}
 isMobileLOD {type=boolean}
 isMobileLODAlpha {type=number}
 isNotAimingAlpha {type=number}
 isNotCrouchingAlpha {type=number}
 isNotCrouchingJacketAlpha_08 {type=number}
 isNotSkydiving {type=number}
 isSkydiveFloatingAlpha {type=number}
 isSkydiveNotFloating {type=number}
 isTargetingAlpha {type=number}
 low fps {type=number}
 R_legSwingAlpha {type=number}
 spineSwingAngleAlpha {type=number}
 thigh L Rot {type=number}
 Thigh R Rot {type=number}
 thigh_rot_inr_l {type=number}
 thigh_rot_inr_r {type=number}
 thigh_rot_otr_l {type=number}
 thigh_rot_otr_r {type=number}
 thigh_rot_up_l {type=number}
 thigh_rot_up_r {type=number}
 windVector {type=Vector}
]
