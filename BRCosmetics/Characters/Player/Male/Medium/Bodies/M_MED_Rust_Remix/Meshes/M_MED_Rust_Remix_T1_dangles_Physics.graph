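# Physics asset for the dangling accessories on the M_MED_Rust_Remix_T1 player body: left and right wrist flaps and a right leg pouch.
# Each accessory bone has a capsule body, constrained to its own MSTR bone, which carries only a tiny placeholder-sized box.
# Angular limits let each piece swing mainly on one axis, wider for the flaps and pouch on that axis, with tight twist.
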PhysicsAsset /BRCosmetics/Characters/Player/Male/Medium/Bodies/M_MED_Rust_Remix/Meshes/M_MED_Rust_Remix_T1_dangles_Physics
bodies {
 dyn_legpouch_r_01 [{shapeType=Capsule, length=6.9075, radius=2.5651, shapeName=dyn_legpouch_r_01_capsule, center=<2.8474 0 -0.0003>, rotation=<rot 0 0 90>}]
 dyn_legpouch_r_01_MSTR [{shapeType=Box, extentX=0.01, extentY=0.01, extentZ=0.01}]
 dyn_wristFlap_01_l [{shapeType=Capsule, length=6.0359, radius=1.8633, shapeName=dyn_wristFlap_01_l_capsule, center=<4.1448 0 0>, rotation=<rot 90 0 0>}]
 dyn_wristFlap_01_l_MSTR [{shapeType=Box, extentX=0.01, extentY=0.01, extentZ=0.01}]
 dyn_wristFlap_01_r [{shapeType=Capsule, length=4.8019, radius=1.8047, shapeName=dyn_wristFlap_01_r_capsule, center=<-4.5445 0 0>, rotation=<rot 90 0.0002 0>}]
 dyn_wristFlap_01_r_MSTR [{shapeType=Box, extentX=0.01, extentY=0.01, extentZ=0.01}]
}
constraints [
 {bone1Name=dyn_wristFlap_01_l, bone2Name=dyn_wristFlap_01_l_MSTR, swing1LimitDegrees=45, swing2LimitDegrees=5, twistLimitDegrees=5}
 {bone1Name=dyn_legpouch_r_01, bone2Name=dyn_legpouch_r_01_MSTR, swing1LimitDegrees=15, swing2LimitDegrees=45, twistLimitDegrees=5}
 {bone1Name=dyn_wristFlap_01_r, bone2Name=dyn_wristFlap_01_r_MSTR, swing1LimitDegrees=45, swing2LimitDegrees=5, twistLimitDegrees=5}
]
