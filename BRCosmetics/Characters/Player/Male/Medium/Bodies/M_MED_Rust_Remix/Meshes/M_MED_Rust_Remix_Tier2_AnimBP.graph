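# Animation Blueprint for the Rust Remix Tier2 male body cosmetic, copying the parent mesh pose and adding secondary motion.
# Constraints and flap nodes adjust shoe bones, bone offsets reposition straps by skydive state, and weighted constraints place shoulder pad helpers.
# A RigidBody node then simulates the T1 dangles physics asset, using scaled skydiving force as wind.
# The update event computes these alphas, disabling shoulder constraints above LOD 1 or without rigid body.
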AnimBlueprint /BRCosmetics/Characters/Player/Male/Medium/Bodies/M_MED_Rust_Remix/Meshes/M_MED_Rust_Remix_Tier2_AnimBP
target_skeleton=&/M_MED_Rust_Remix_Skeleton
graphs [
 AnimGraph {nodes [
   n1 = "Constraint - shoe_01_l_DSP (P)"(ComponentPose=Local To Component(LocalPose=Copy Pose From Mesh() {bUseAttachedParent=true}), Alpha=1, ConstraintWeights_0=1) {
    alphaBoolBlend {blendOption=Linear}
    boneToModify {boneName=shoe_01_l_DSP}
    constraintSetup [{targetBone {boneName=calf_twist_02_l}}]
    constraintWeights [1]
   }
   n2 = Flap Node - Bone: shoe_01_l_DSP!(ComponentPose:Component Space Pose Link Structure=@n1, Alpha:float=1) -> (Pose:Component Space Pose Link Structure) {
    is=title FortAnimGraphNode_Flap
    bUseClamp=true
    targetClampMax=1
    targetComponent=RotationZ
    alphaBoolBlend {blendOption=Linear}
    sourceDrives [{
      bUseRange=false
      drivingCurve=null
      multiplier=0.8
      rangeMax=1
      rangeMin=-1
      remappedMax=1
      remappedMin=0
      sourceComponent=RotationZ
      useQuaternion=true
      sourceBone {boneName=foot_l}
     }]
    targetBone {boneName=shoe_01_l_DSP}
   }
   n3 = "Constraint - shoe_01_r_DSP (P)"(ComponentPose=@n2, Alpha=1, ConstraintWeights_0=1) {
    alphaBoolBlend {blendOption=Linear}
    boneToModify {boneName=shoe_01_r_DSP}
    constraintSetup [{targetBone {boneName=calf_twist_02_r}}]
    constraintWeights [1]
   }
   n4 = Flap Node - Bone: shoe_01_r_DSP!(ComponentPose:Component Space Pose Link Structure=@n3, Alpha:float=1) -> (Pose:Component Space Pose Link Structure) {
    is=title FortAnimGraphNode_Flap
    bUseClamp=true
    targetClampMax=1
    targetComponent=RotationZ
    alphaBoolBlend {blendOption=Linear}
    sourceDrives [{
      bUseRange=false
      drivingCurve=null
      multiplier=0.8
      rangeMax=1
      rangeMin=-1
      remappedMax=1
      remappedMin=0
      sourceComponent=RotationZ
      useQuaternion=true
      sourceBone {boneName=foot_r}
     }]
    targetBone {boneName=shoe_01_r_DSP}
   }
   n5 = "Transform (Modify) Bone - Bone: strap_back_01_DSP"(
    ComponentPose=@n4
    Alpha=Get isSkydiveFloatingAlpha() -> (isSkydiveFloatingAlpha:double)
    Translation="0,-2.5,2.5"
    Rotation="0,0,0"
    Scale="1,1,1"
   ) {
    rotationMode=BMM_Additive
    rotationSpace=BCS_BoneSpace
    scaleMode=BMM_Additive
    scaleSpace=BCS_BoneSpace
    translationMode=BMM_Additive
    translationSpace=BCS_BoneSpace
    alphaBoolBlend {blendOption=Linear}
    boneToModify {boneName=strap_back_01_DSP}
    translation {y=-2.5, z=2.5}
   }
   n6 = "Transform (Modify) Bone - Bone: strap_back_01_DSP"(ComponentPose=@n5, Alpha=Get isSkydiveNotFloating() -> (isSkydiveNotFloating:double), Translation="0,3,0", Rotation="0,0,0", Scale="1,1,1") {
    rotationMode=BMM_Additive
    rotationSpace=BCS_BoneSpace
    scaleMode=BMM_Additive
    scaleSpace=BCS_BoneSpace
    translationMode=BMM_Additive
    translationSpace=BCS_BoneSpace
    alphaBoolBlend {blendOption=Linear}
    boneToModify {boneName=strap_back_01_DSP}
    translation {y=3}
   }
   n7 = "Transform (Modify) Bone - Bone: strap_front_01_DSP"(ComponentPose=@n6, Alpha=1, Translation="0,0,2", Rotation="0,0,0", Scale="1,1,1") {
    rotationMode=BMM_Additive
    rotationSpace=BCS_BoneSpace
    scaleMode=BMM_Additive
    scaleSpace=BCS_BoneSpace
    translationMode=BMM_Additive
    translationSpace=BCS_BoneSpace
    alphaBoolBlend {blendOption=Linear}
    boneToModify {boneName=strap_front_01_DSP}
    translation {z=2}
   }
   n8 = Get isNotSkydiving() -> (isNotSkydiving:double)
   n9 = "Constraint - strap_back_01_DSP (PP)"!(ComponentPose=@n7, Alpha=@n8, ConstraintWeights_0=0.5, ConstraintWeights_1:float=0.5) -> (Pose) {
    alphaBoolBlend {blendOption=Linear}
    boneToModify {boneName=strap_back_01_DSP}
    constraintSetup [{targetBone {boneName=shoulderBodyArmor_back_01_l_DSP}}, {targetBone {boneName=spine_03}}]
    constraintWeights [1,1]
   }
   n10 = "Transform (Modify) Bone - Bone: strap_back_01_DSP"(ComponentPose=@n9, Alpha=@n8, Translation="0,0,2.5", Rotation="0,0,0", Scale="1,1,1") {
    rotationMode=BMM_Additive
    rotationSpace=BCS_BoneSpace
    scaleMode=BMM_Additive
    scaleSpace=BCS_BoneSpace
    translationMode=BMM_Additive
    translationSpace=BCS_BoneSpace
    alphaBoolBlend {blendOption=Linear}
    boneToModify {boneName=strap_back_01_DSP}
    translation {z=2.5}
   }
   n11 = Get isMobileLODAlphaInv() -> (isMobileLODAlphaInv:double)
   n12 = "Constraint - dyn_shldr_frt_r (PPPPPPP)"!(
    ComponentPose=@n10
    Alpha=@n11
    ConstraintWeights_0=0.0663
    ConstraintWeights_1:float=0.6111
    ConstraintWeights_2:float=0.1721
    ConstraintWeights_3:float=0.0108
    ConstraintWeights_4:float=0.0245
    ConstraintWeights_5:float=0.0101
    ConstraintWeights_6:float=0.1051
   ) -> (Pose) {
    alphaBoolBlend {blendOption=Linear}
    boneToModify {boneName=dyn_shldr_frt_r}
    constraintSetup [
     {targetBone {boneName=upperarm_r}}
     {targetBone {boneName=upperarm_twist_01_r}}
     {targetBone {boneName=upperarm_twist_02_r}}
     {targetBone {boneName=spine_02}}
     {targetBone {boneName=spine_03}}
     {targetBone {boneName=spine_04}}
     {targetBone {boneName=clavicle_r}}
    ]
    constraintWeights [1,1,1,1,1,1,1]
   }
   n13 = "Constraint - dyn_shldr_bck_r (PPPPP)"!(
    ComponentPose=@n12
    Alpha=@n11
    ConstraintWeights_0=0.3041
    ConstraintWeights_1:float=0.0671
    ConstraintWeights_2:float=0.0361
    ConstraintWeights_3:float=0.0963
    ConstraintWeights_4:float=0.4964
   ) -> (Pose) {
    alphaBoolBlend {blendOption=Linear}
    boneToModify {boneName=dyn_shldr_bck_r}
    constraintSetup [
     {targetBone {boneName=clavicle_r}}
     {targetBone {boneName=upperarm_r}}
     {targetBone {boneName=spine_03}}
     {targetBone {boneName=upperarm_twist_02_r}}
     {targetBone {boneName=upperarm_twist_01_r}}
    ]
    constraintWeights [1,1,1,1,1]
   }
   n14 = "Constraint - dyn_parent_shldr_pad_r (PP)"!(ComponentPose=@n13, Alpha=@n11, ConstraintWeights_0=0.3, ConstraintWeights_1:float=0.7) -> (Pose) {
    alphaBoolBlend {blendOption=Linear}
    boneToModify {boneName=dyn_parent_shldr_pad_r}
    constraintSetup [{targetBone {boneName=dyn_shldr_frt_r}}, {targetBone {boneName=dyn_shldr_bck_r}}]
    constraintWeights [1,1]
   }
   n15 = "Constraint - dyn_shldr_frt_l (PPPPPPPP)"!(
    ComponentPose=@n14
    Alpha=@n11
    ConstraintWeights_0=0.2077
    ConstraintWeights_1:float=0.2978
    ConstraintWeights_2:float=0.0148
    ConstraintWeights_3:float=0.0344
    ConstraintWeights_4:float=0.02
    ConstraintWeights_5:float=0.0217
    ConstraintWeights_6:float=0.33
    ConstraintWeights_7:float=0.0736
   ) -> (Pose) {
    alphaBoolBlend {blendOption=Linear}
    boneToModify {boneName=dyn_shldr_frt_l}
    constraintSetup [
     {targetBone {boneName=clavicle_l}}
     {targetBone {boneName=upperarm_twist_02_l}}
     {targetBone {boneName=spine_02}}
     {targetBone {boneName=spine_03}}
     {targetBone {boneName=spine_04}}
     {targetBone {boneName=spine_05}}
     {targetBone {boneName=upperarm_twist_01_l}}
     {targetBone {boneName=upperarm_l}}
    ]
    constraintWeights [1,1,1,1,1,1,1,1]
   }
   n16 = "Constraint - dyn_shldr_bck_l (PPPPP)"!(
    ComponentPose=@n15
    Alpha=@n11
    ConstraintWeights_0=0.0876
    ConstraintWeights_1:float=0.0377
    ConstraintWeights_2:float=0.4933
    ConstraintWeights_3:float=0.3207
    ConstraintWeights_4:float=0.0606
   ) -> (Pose) {
    alphaBoolBlend {blendOption=Linear}
    boneToModify {boneName=dyn_shldr_bck_l}
    constraintSetup [
     {targetBone {boneName=upperarm_twist_02_l}}
     {targetBone {boneName=spine_03}}
     {targetBone {boneName=upperarm_twist_01_l}}
     {targetBone {boneName=clavicle_l}}
     {targetBone {boneName=upperarm_l}}
    ]
    constraintWeights [1,1,1,1,1]
   }
   n17 = "Constraint - dyn_parent_shldr_pad_l (PP)"!(ComponentPose=@n16, Alpha=@n11, ConstraintWeights_0=0.5, ConstraintWeights_1:float=0.5) -> (Pose) {
    alphaBoolBlend {blendOption=Linear}
    boneToModify {boneName=dyn_parent_shldr_pad_l}
    constraintSetup [{targetBone {boneName=dyn_shldr_frt_l}}, {targetBone {boneName=dyn_shldr_bck_l}}]
    constraintWeights [1,1]
   }
   n18 = RigidBody(ComponentPose=@n17, ExternalForce=Get windVector() -> (windVector:Vector)) {
    bClampLinearTranslationLimitToRefPose=true
    bForceDisableCollisionBetweenConstraintBodies=true
    bOverrideWorldGravity=true
    overridePhysicsAsset=&/M_MED_Rust_Remix_T1_dangles_Physics
    simulationSpace=BaseBoneSpace
    alphaBoolBlend {blendOption=Linear}
    baseBoneRef {boneName=spine_01}
    componentAppliedLinearAccClamp=<3000 3000 3000>
    componentLinearAccScale=<0.3 0.3 0.6>
    componentLinearVelScale=<0.3 0.3 0.6>
    overrideWorldGravity {z=-980}
   }
   Output Pose(Result=Component To Local(ComponentPose=@n18)) {layerGroup=None}
  ]}
 EventGraph {nodes [
   Event BlueprintUpdateAnimation()
   > n1 = Sequence!(execute) -> (then_0, then_1, then_2:Exec, then_3:Exec)
   n2 = Get bIsSkydiveFloating()
   > Set isSkydiveFloatingAlpha(isSkydiveFloatingAlpha:double="To Float (Boolean)"(InBool=@n2)) -> (Output_Get:double)
   n3 = NOT Boolean(A=Get bIsSkydiveDiveMode())
   n4 = NOT Boolean(A=@n2)
   n5 = Get bIsSkyDiving()
   > Set isSkydiveNotFloating(isSkydiveNotFloating:double="To Float (Boolean)"(InBool=AND Boolean!(A=@n4, B=@n5, C:bool=@n3) -> (ReturnValue))) -> (Output_Get:double)
   n6 = Get bIsSkyDiving()
   > Set isNotSkydiving(isNotSkydiving:double="To Float (Boolean)"(InBool=NOT Boolean(A=@n6))) -> (Output_Get:double)
   n7 = Get SkydivingExternalForceInput()
   Set windVector(
    execute=@n1.then_1
    windVector:Vector=vector * float(
     A=Break Fort Anim Input Skydiving External Force(FortAnimInput_SkydivingExternalForce=@n7) -> (
      bUseSkydivingVectorForce=~
      SkydivingVectorForce=~
      HeadToPelvisDirection=~
      FloatingMultiplier=~
      FloatingAdditive=~
      DivingMultiplier=~
      DivingAdditive=~
      ParachutingMultiplier=~
      ParachutingAdditive=~
     ).SkydivingVectorForce
     B=3
    )
   ) -> (Output_Get:Vector)
   n8 = Get CurrentLOD()
   n9 = Get bRigidBodyEnabled()
   n10 = Set isMobileLOD(execute=@n1.then_2, isMobileLOD:bool=OR Boolean(A=NOT Boolean(A=@n9), B="integer > integer"(A=@n8, B=1) {is=title K2Node_CallFunction})) -> (Output_Get:bool)
   > Set isMobileLODAlpha(isMobileLODAlpha:double="To Float (Boolean)"(InBool=@n10)) -> (Output_Get:double)
   > Set isMobileLODAlphaInv(isMobileLODAlphaInv:double="To Float (Boolean)"(InBool=NOT Boolean(A=@n10))) -> (Output_Get:double)
  ]}
]
variables [
 isMobileLOD {type=boolean}
 isMobileLODAlpha {type=number}
 isMobileLODAlphaInv {type=number}
 isNotSkydiving {type=number}
 isSkydiveFloatingAlpha {type=number}
 isSkydiveNotFloating {type=number}
 windVector {type=Vector}
]
